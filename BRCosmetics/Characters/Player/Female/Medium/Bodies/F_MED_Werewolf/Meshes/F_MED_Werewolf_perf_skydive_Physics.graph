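# Physics asset for the skydive variant of the female medium Werewolf cosmetic body, simulating a jacket against lower-body capsules on the pelvis, thighs, calves and left foot.
# Six jacket chains (front, side, back, left and right) hang from the pelvis via angularly free master bones, each with five swing- and twist-limited segments.
# Angularly free cross-constraints tie neighboring chains at the third and fourth segments, leaving the front open.
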PhysicsAsset /BRCosmetics/Characters/Player/Female/Medium/Bodies/F_MED_Werewolf/Meshes/F_MED_Werewolf_perf_skydive_Physics
bodies {
 calf_l [{shapeType=Capsule, length=39.232, radius=18.1782, center=<-28.652 5.3252 0.0466>, rotation=<rot 0 89.533 -88.5703>}]
 calf_r [{shapeType=Capsule, length=39.232, radius=18.1782, center=<28.6861 -4.1708 1.5067>, rotation=<rot -0.3385 89.533 -88.4634>}]
 dyn_jacketBack_01_l [{shapeType=Capsule, length=8.2952, radius=7.5596, shapeName=dyn_jacketBack_01_l_capsule, center=<9.3842 1.5 6>, rotation=<rot 0 0 87.8776>}]
 dyn_jacketBack_01_r [{shapeType=Capsule, length=6.9724, radius=8.3191, shapeName=dyn_jacketBack_01_r_capsule, center=<8.5321 -1 7>, rotation=<rot 0 0 106.6011>}]
 dyn_jacketBack_02_l [{shapeType=Capsule, length=11.2569, radius=8.9034, shapeName=dyn_jacketBack_02_l_capsule, center=<9.0158 3.5 5>, rotation=<rot 0 0 97.2548>}]
 dyn_jacketBack_02_r [{
   shapeType=Capsule
   length=11.2131
   radius=9.0594
   shapeName=dyn_jacketBack_02_r_capsule
   center=<8.6777 0 3.5>
   rotation=<rot 0.2239 -1.5154 98.4019>
  }]
 dyn_jacketBack_03_l [{shapeType=Capsule, length=10.7541, radius=9.26, shapeName=dyn_jacketBack_03_l_capsule, center=<8.91 4 4.5>, rotation=<rot 0 0 103.6992>}]
 dyn_jacketBack_03_r [{
   shapeType=Capsule
   length=19.4472
   radius=9.26
   shapeName=dyn_jacketBack_03_r_capsule
   center=<8.6464 -0.5 5>
   rotation=<rot 0 -6.6638 89.9997>
  }]
 dyn_jacketBack_04_l [{shapeType=Capsule, length=15.527, radius=10.6603, shapeName=dyn_jacketBack_04_l_capsule, center=<7.6587 8 4>, rotation=<rot 0 0 103.2565>}]
 dyn_jacketBack_04_r [{
   shapeType=Capsule
   length=23.1564
   radius=9.1117
   shapeName=dyn_jacketBack_04_r_capsule
   center=<6.1458 -5 6>
   rotation=<rot 0 -8.2266 89.4698>
  }]
 dyn_jacketBack_05_l [{
   shapeType=Capsule
   length=11.2318
   radius=13.001
   shapeName=dyn_jacketBack_05_l_capsule
   center=<9.2035 7 3.5>
   rotation=<rot 0 9.8043 99.2008>
  }]
 dyn_jacketBack_05_r [{
   shapeType=Capsule
   length=9.8936
   radius=14.3393
   shapeName=dyn_jacketBack_05_r_capsule
   center=<9.3028 -8 8>
   rotation=<rot 0 -12.0502 89.9999>
  }]
 dyn_jacketBack_mstr_l [{shapeType=Capsule, length=0.505, radius=0.505, shapeName=dyn_jacketBack_mstr_l_capsule, rotation=<rot 0 -90 0>}]
 dyn_jacketBack_mstr_r [{shapeType=Capsule, length=0.505, radius=0.505, shapeName=dyn_jacketBack_mstr_r_capsule, rotation=<rot 0 -90 0>}]
 dyn_jacketFront_01_l [{shapeType=Capsule, length=6.3575, radius=8.7802, shapeName=dyn_jacketFront_01_l_capsule, center=<7.7894 0 8.5>, rotation=<rot 0 0 99.079>}]
 dyn_jacketFront_01_r [{shapeType=Capsule, length=9.822, radius=8.2496, shapeName=dyn_jacketFront_01_r_capsule, center=<9.3537 0 8>, rotation=<rot 0 0 65.6585>}]
 dyn_jacketFront_02_l [{shapeType=Capsule, length=8.0596, radius=9.1589, shapeName=dyn_jacketFront_02_l_capsule, center=<9.3 0 8>, rotation=<rot 0 0 98.9778>}]
 dyn_jacketFront_02_r [{shapeType=Capsule, length=10.2067, radius=8.7423, shapeName=dyn_jacketFront_02_r_capsule, center=<9.3231 -0.5 7.5>, rotation=<rot 0 0 65.64>}]
 dyn_jacketFront_03_l [{shapeType=Capsule, length=4.5068, radius=8.4427, shapeName=dyn_jacketFront_03_l_capsule, center=<9.0617 2 5.25>, rotation=<rot 0 0 102.6331>}]
 dyn_jacketFront_03_r [{shapeType=Capsule, length=21.2132, radius=8.5543, shapeName=dyn_jacketFront_03_r_capsule, center=<9.0617 -2 5>, rotation=<rot 0 0 62.4186>}]
 dyn_jacketFront_04_l [{
   shapeType=Capsule
   length=8.4845
   radius=8.7521
   shapeName=dyn_jacketFront_04_l_capsule
   center=<8.4824 0.5 3.5>
   rotation=<rot 0 0 100.7803>
  }]
 dyn_jacketFront_04_r [{shapeType=Capsule, length=21.7643, radius=8.7521, shapeName=dyn_jacketFront_04_r_capsule, center=<8.4823 -2 4.5>, rotation=<rot 0 0 65.629>}]
 dyn_jacketFront_05_l [{
   shapeType=Capsule
   length=10.673
   radius=9.8625
   shapeName=dyn_jacketFront_05_l_capsule
   center=<10 3.5 2>
   rotation=<rot 3.6898 -12.0755 106.7429>
  }]
 dyn_jacketFront_05_r [{shapeType=Capsule, length=21.1086, radius=9.8625, shapeName=dyn_jacketFront_05_r_capsule, center=<10 -3.5 2>, rotation=<rot 0 0 62.3001>}]
 dyn_jacketFront_mstr_l [{shapeType=Capsule, length=0.505, radius=0.505, shapeName=dyn_jacketFront_mstr_l_capsule, rotation=<rot 0 -90 0>}]
 dyn_jacketFront_mstr_r [{shapeType=Capsule, length=0.505, radius=0.505, shapeName=dyn_jacketFront_mstr_r_capsule, rotation=<rot 90 0 0>}]
 dyn_jacketSide_01_l [{shapeType=Capsule, length=7.6208, radius=8.0702, shapeName=dyn_jacketSide_01_l_capsule, center=<9.0875 -0.5 7.5>, rotation=<rot 0 0 62.5433>}]
 dyn_jacketSide_01_r [{shapeType=Capsule, length=9.3643, radius=6.6624, shapeName=dyn_jacketSide_01_r_capsule, center=<8.8641 3.5 4>, rotation=<rot 0 0 125.9463>}]
 dyn_jacketSide_02_l [{shapeType=Capsule, length=11.8347, radius=8.0561, shapeName=dyn_jacketSide_02_l_capsule, center=<9.0207 -1 6.5>, rotation=<rot 0 0 64.7146>}]
 dyn_jacketSide_02_r [{shapeType=Capsule, length=22.2934, radius=8.106, shapeName=dyn_jacketSide_02_r_capsule, center=<8.4512 2.5 4>, rotation=<rot 0 0 122.6685>}]
 dyn_jacketSide_03_l [{shapeType=Capsule, length=8.4279, radius=8.3669, shapeName=dyn_jacketSide_03_l_capsule, center=<9.1814 0.5 5.5>, rotation=<rot 0 0 73.7319>}]
 dyn_jacketSide_03_r [{shapeType=Capsule, length=27.5066, radius=8.7392, shapeName=dyn_jacketSide_03_r_capsule, center=<8.2304 3.5 4>, rotation=<rot 0 0 122.5465>}]
 dyn_jacketSide_04_l [{shapeType=Capsule, length=10.7585, radius=9.2843, shapeName=dyn_jacketSide_04_l_capsule, center=<9.4532 0 7>, rotation=<rot 0 0 78.9566>}]
 dyn_jacketSide_04_r [{
   shapeType=Capsule
   length=30.6627
   radius=8.9878
   shapeName=dyn_jacketSide_04_r_capsule
   center=<8.8967 3.5 5.5>
   rotation=<rot 0 0 115.7065>
  }]
 dyn_jacketSide_05_l [{shapeType=Capsule, length=18.1014, radius=9.5621, shapeName=dyn_jacketSide_05_l_capsule, center=<9.2752 2.5 6>, rotation=<rot 0 0 77.1268>}]
 dyn_jacketSide_05_r [{shapeType=Capsule, length=21.1676, radius=9.969, shapeName=dyn_jacketSide_05_r_capsule, center=<9.1948 3.5 5>, rotation=<rot 0 0 114.7002>}]
 dyn_jacketSide_mstr_l [{shapeType=Capsule, length=0.505, radius=0.505, shapeName=dyn_jacketSide_mstr_l_capsule, rotation=<rot 0 -90 0>}]
 dyn_jacketSide_mstr_r [{shapeType=Capsule, length=0.505, radius=0.505, shapeName=dyn_jacketSide_mstr_r_capsule, rotation=<rot 0 0 90>}]
 foot_l [{
   shapeType=Capsule
   length=0.7412
   radius=16.0629
   shapeName=foot_l_capsule
   center=<-4.7664 3.0033 -0.4447>
   rotation=<rot 0 -151.9071 -84.6337>
  }]
 pelvis [
  {shapeType=Capsule, length=9.0875, radius=11, center=<-0.9227 -1.2812 0>}
  {shapeType=Capsule, length=13.415, radius=7.4702, shapeName=pelvis_capsule, center=<-3.9978 -3.7131 0>}
 ]
 thigh_l [{shapeType=Capsule, length=26.6217, radius=10.7144, center=<-21.4655 2.2721 -0.2548>, rotation=<rot 90 -179.8881 -179.6926>}]
 thigh_r [{shapeType=Capsule, length=25.3656, radius=11.7144, center=<21.7868 -2.0701 -1.1637>, rotation=<rot 88.343 -180 -180>}]
}
constraints [
 {bone1Name=thigh_l, bone2Name=pelvis, swing1LimitDegrees=45, swing2LimitDegrees=45, twistLimitDegrees=45}
 {bone1Name=calf_l, bone2Name=thigh_l, swing1LimitDegrees=45, swing2LimitDegrees=45, twistLimitDegrees=45}
 {bone1Name=thigh_r, bone2Name=pelvis, swing1LimitDegrees=45, swing2LimitDegrees=45, twistLimitDegrees=45}
 {bone1Name=calf_r, bone2Name=thigh_r, swing1LimitDegrees=45, swing2LimitDegrees=45, twistLimitDegrees=45}
 {
  bone1Name=dyn_jacketFront_mstr_r
  bone2Name=pelvis
  swing1LimitDegrees=45
  swing1Motion=Free
  swing2LimitDegrees=45
  swing2Motion=Free
  twistLimitDegrees=45
  twistMotion=Free
 }
 {bone1Name=dyn_jacketFront_01_r, bone2Name=dyn_jacketFront_mstr_r, swing1LimitDegrees=35, swing2LimitDegrees=35, twistLimitDegrees=15}
 {bone1Name=dyn_jacketFront_02_r, bone2Name=dyn_jacketFront_01_r, swing1LimitDegrees=35, swing2LimitDegrees=35, twistLimitDegrees=15}
 {bone1Name=dyn_jacketFront_03_r, bone2Name=dyn_jacketFront_02_r, swing1LimitDegrees=35, swing2LimitDegrees=35, twistLimitDegrees=15}
 {bone1Name=dyn_jacketFront_04_r, bone2Name=dyn_jacketFront_03_r, swing1LimitDegrees=35, swing2LimitDegrees=35, twistLimitDegrees=15}
 {bone1Name=dyn_jacketFront_05_r, bone2Name=dyn_jacketFront_04_r, swing1LimitDegrees=35, swing2LimitDegrees=35, twistLimitDegrees=15}
 {
  bone1Name=dyn_jacketFront_mstr_l
  bone2Name=pelvis
  swing1LimitDegrees=45
  swing1Motion=Free
  swing2LimitDegrees=45
  swing2Motion=Free
  twistLimitDegrees=45
  twistMotion=Free
 }
 {bone1Name=dyn_jacketFront_01_l, bone2Name=dyn_jacketFront_mstr_l, swing1LimitDegrees=35, swing2LimitDegrees=35, twistLimitDegrees=15}
 {bone1Name=dyn_jacketFront_02_l, bone2Name=dyn_jacketFront_01_l, swing1LimitDegrees=35, swing2LimitDegrees=35, twistLimitDegrees=15}
 {bone1Name=dyn_jacketFront_03_l, bone2Name=dyn_jacketFront_02_l, swing1LimitDegrees=35, swing2LimitDegrees=35, twistLimitDegrees=15}
 {bone1Name=dyn_jacketFront_04_l, bone2Name=dyn_jacketFront_03_l, swing1LimitDegrees=35, swing2LimitDegrees=35, twistLimitDegrees=15}
 {bone1Name=dyn_jacketFront_05_l, bone2Name=dyn_jacketFront_04_l, swing1LimitDegrees=35, swing2LimitDegrees=35, twistLimitDegrees=15}
 {
  bone1Name=dyn_jacketSide_mstr_r
  bone2Name=pelvis
  swing1LimitDegrees=45
  swing1Motion=Free
  swing2LimitDegrees=45
  swing2Motion=Free
  twistLimitDegrees=45
  twistMotion=Free
 }
 {bone1Name=dyn_jacketSide_01_r, bone2Name=dyn_jacketSide_mstr_r, swing1LimitDegrees=35, swing2LimitDegrees=35, twistLimitDegrees=15}
 {bone1Name=dyn_jacketSide_02_r, bone2Name=dyn_jacketSide_01_r, swing1LimitDegrees=35, swing2LimitDegrees=35, twistLimitDegrees=15}
 {bone1Name=dyn_jacketSide_03_r, bone2Name=dyn_jacketSide_02_r, swing1LimitDegrees=35, swing2LimitDegrees=35, twistLimitDegrees=15}
 {bone1Name=dyn_jacketSide_04_r, bone2Name=dyn_jacketSide_03_r, swing1LimitDegrees=35, swing2LimitDegrees=35, twistLimitDegrees=15}
 {bone1Name=dyn_jacketSide_05_r, bone2Name=dyn_jacketSide_04_r, swing1LimitDegrees=35, swing2LimitDegrees=35, twistLimitDegrees=15}
 {
  bone1Name=dyn_jacketBack_mstr_r
  bone2Name=pelvis
  swing1LimitDegrees=45
  swing1Motion=Free
  swing2LimitDegrees=45
  swing2Motion=Free
  twistLimitDegrees=45
  twistMotion=Free
 }
 {bone1Name=dyn_jacketBack_01_r, bone2Name=dyn_jacketBack_mstr_r, swing1LimitDegrees=35, swing2LimitDegrees=35, twistLimitDegrees=15}
 {bone1Name=dyn_jacketBack_02_r, bone2Name=dyn_jacketBack_01_r, swing1LimitDegrees=35, swing2LimitDegrees=35, twistLimitDegrees=15}
 {bone1Name=dyn_jacketBack_03_r, bone2Name=dyn_jacketBack_02_r, swing1LimitDegrees=35, swing2LimitDegrees=35, twistLimitDegrees=15}
 {bone1Name=dyn_jacketBack_04_r, bone2Name=dyn_jacketBack_03_r, swing1LimitDegrees=35, swing2LimitDegrees=35, twistLimitDegrees=15}
 {bone1Name=dyn_jacketBack_05_r, bone2Name=dyn_jacketBack_04_r, swing1LimitDegrees=35, swing2LimitDegrees=35, twistLimitDegrees=15}
 {
  bone1Name=dyn_jacketBack_mstr_l
  bone2Name=pelvis
  swing1LimitDegrees=45
  swing1Motion=Free
  swing2LimitDegrees=45
  swing2Motion=Free
  twistLimitDegrees=45
  twistMotion=Free
 }
 {bone1Name=dyn_jacketBack_01_l, bone2Name=dyn_jacketBack_mstr_l, swing1LimitDegrees=35, swing2LimitDegrees=35, twistLimitDegrees=15}
 {bone1Name=dyn_jacketBack_02_l, bone2Name=dyn_jacketBack_01_l, swing1LimitDegrees=35, swing2LimitDegrees=35, twistLimitDegrees=15}
 {bone1Name=dyn_jacketBack_03_l, bone2Name=dyn_jacketBack_02_l, swing1LimitDegrees=35, swing2LimitDegrees=35, twistLimitDegrees=15}
 {bone1Name=dyn_jacketBack_04_l, bone2Name=dyn_jacketBack_03_l, swing1LimitDegrees=35, swing2LimitDegrees=35, twistLimitDegrees=15}
 {bone1Name=dyn_jacketBack_05_l, bone2Name=dyn_jacketBack_04_l, swing1LimitDegrees=35, swing2LimitDegrees=35, twistLimitDegrees=15}
 {
  bone1Name=dyn_jacketSide_mstr_l
  bone2Name=pelvis
  swing1LimitDegrees=45
  swing1Motion=Free
  swing2LimitDegrees=45
  swing2Motion=Free
  twistLimitDegrees=45
  twistMotion=Free
 }
 {bone1Name=dyn_jacketSide_01_l, bone2Name=dyn_jacketSide_mstr_l, swing1LimitDegrees=35, swing2LimitDegrees=35, twistLimitDegrees=15}
 {bone1Name=dyn_jacketSide_02_l, bone2Name=dyn_jacketSide_01_l, swing1LimitDegrees=35, swing2LimitDegrees=35, twistLimitDegrees=15}
 {bone1Name=dyn_jacketSide_03_l, bone2Name=dyn_jacketSide_02_l, swing1LimitDegrees=35, swing2LimitDegrees=35, twistLimitDegrees=15}
 {bone1Name=dyn_jacketSide_04_l, bone2Name=dyn_jacketSide_03_l, swing1LimitDegrees=35, swing2LimitDegrees=35, twistLimitDegrees=15}
 {bone1Name=dyn_jacketSide_05_l, bone2Name=dyn_jacketSide_04_l, swing1LimitDegrees=35, swing2LimitDegrees=35, twistLimitDegrees=15}
 {
  bone1Name=dyn_jacketBack_03_l
  bone2Name=dyn_jacketBack_03_r
  swing1LimitDegrees=45
  swing1Motion=Free
  swing2LimitDegrees=45
  swing2Motion=Free
  twistLimitDegrees=45
  twistMotion=Free
 }
 {
  bone1Name=dyn_jacketBack_03_r
  bone2Name=dyn_jacketSide_03_r
  swing1LimitDegrees=45
  swing1Motion=Free
  swing2LimitDegrees=45
  swing2Motion=Free
  twistLimitDegrees=45
  twistMotion=Free
 }
 {
  bone1Name=foot_l
  bone2Name=calf_l
  swing1LimitDegrees=45
  swing1Motion=Free
  swing2LimitDegrees=45
  swing2Motion=Free
  twistLimitDegrees=45
  twistMotion=Free
 }
 {
  bone1Name=dyn_jacketSide_03_l
  bone2Name=dyn_jacketBack_03_l
  swing1LimitDegrees=45
  swing1Motion=Free
  swing2LimitDegrees=45
  swing2Motion=Free
  twistLimitDegrees=45
  twistMotion=Free
 }
 {
  bone1Name=dyn_jacketSide_03_r
  bone2Name=dyn_jacketBack_03_r
  swing1LimitDegrees=45
  swing1Motion=Free
  swing2LimitDegrees=45
  swing2Motion=Free
  twistLimitDegrees=45
  twistMotion=Free
 }
 {
  bone1Name=dyn_jacketFront_03_r
  bone2Name=dyn_jacketSide_03_r
  swing1LimitDegrees=45
  swing1Motion=Free
  swing2LimitDegrees=45
  swing2Motion=Free
  twistLimitDegrees=45
  twistMotion=Free
 }
 {
  bone1Name=dyn_jacketFront_03_l
  bone2Name=dyn_jacketSide_03_l
  swing1LimitDegrees=45
  swing1Motion=Free
  swing2LimitDegrees=45
  swing2Motion=Free
  twistLimitDegrees=45
  twistMotion=Free
 }
 {
  bone1Name=dyn_jacketBack_04_l
  bone2Name=dyn_jacketBack_04_r
  swing1LimitDegrees=45
  swing1Motion=Free
  swing2LimitDegrees=45
  swing2Motion=Free
  twistLimitDegrees=45
  twistMotion=Free
 }
 {
  bone1Name=dyn_jacketSide_04_l
  bone2Name=dyn_jacketBack_04_l
  swing1LimitDegrees=45
  swing1Motion=Free
  swing2LimitDegrees=45
  swing2Motion=Free
  twistLimitDegrees=45
  twistMotion=Free
 }
 {
  bone1Name=dyn_jacketFront_04_l
  bone2Name=dyn_jacketSide_04_l
  swing1LimitDegrees=45
  swing1Motion=Free
  swing2LimitDegrees=45
  swing2Motion=Free
  twistLimitDegrees=45
  twistMotion=Free
 }
 {
  bone1Name=dyn_jacketFront_04_r
  bone2Name=dyn_jacketSide_04_r
  swing1LimitDegrees=45
  swing1Motion=Free
  swing2LimitDegrees=45
  swing2Motion=Free
  twistLimitDegrees=45
  twistMotion=Free
 }
 {
  bone1Name=dyn_jacketSide_04_r
  bone2Name=dyn_jacketBack_04_r
  swing1LimitDegrees=45
  swing1Motion=Free
  swing2LimitDegrees=45
  swing2Motion=Free
  twistLimitDegrees=45
  twistMotion=Free
 }
]
