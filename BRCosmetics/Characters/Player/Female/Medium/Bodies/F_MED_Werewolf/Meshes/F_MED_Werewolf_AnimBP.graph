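# Animation Blueprint for the female medium Werewolf cosmetic body, handling secondary coat motion.
# Chained multi-target Constraint nodes take the cached 'post rigid' pose and drive dyn_coat_extra helper bones as weighted blends of jacket, spine and pelvis bones.
# An AnimDynamics spring chain for those bones appears unconnected, as do some duplicate constraints.
# Modify Bone nodes add crouch-run-dependent jacket rotation and place floorCOL from slope variables.
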AnimBlueprint /BRCosmetics/Characters/Player/Female/Medium/Bodies/F_MED_Werewolf/Meshes/F_MED_Werewolf_AnimBP
target_skeleton=&/F_MED_Werewolf_Skeleton
graphs [
 AnimGraph {nodes [
   n1 = Anim Dynamics - Bone: dyn_coat_extra_01(Alpha=1, ExternalForce="0,0,0") {
    angularDampingOverride=0.9999
    angularSpringConstant=100
    bAngularSpring=true
    bGravityOverrideInSimSpace=true
    bOverrideAngularDamping=true
    simulationSpace=RootRelative
    alphaBoolBlend {blendOption=Linear}
    boundBone {boneName=dyn_coat_extra_01}
    componentAppliedLinearAccClamp=<3500 3500 3500>
    componentLinearAccScale=<1.2 1.2 1.2>
    componentLinearVelScale=<1.2 1.2 1.2>
    physicsBodyDefinitions [{
      sphereCollisionRadius=0
      boundBone {boneName=dyn_coat_extra_01}
      boxExtents=<20 10 10>
      constraintSetup {angularLimitsMax=<1 10 15>, angularLimitsMin=<-1 -30 -15>, angularTarget=<0 0 0>}
      localJointOffset=<2 0 0>
     }]
    relativeSpaceBone {boneName=dyn_root_simspace}
   }
   n2 = Anim Dynamics - Bone: dyn_coat_extra_02(ComponentPose=@n1, Alpha=1, ExternalForce="0,0,0") {
    angularDampingOverride=0.9999
    angularSpringConstant=100
    bAngularSpring=true
    bGravityOverrideInSimSpace=true
    bOverrideAngularDamping=true
    simulationSpace=RootRelative
    alphaBoolBlend {blendOption=Linear}
    boundBone {boneName=dyn_coat_extra_02}
    componentAppliedLinearAccClamp=<3500 3500 3500>
    componentLinearAccScale=<1.2 1.2 1.2>
    componentLinearVelScale=<1.2 1.2 1.2>
    physicsBodyDefinitions [{
      sphereCollisionRadius=0
      boundBone {boneName=dyn_coat_extra_02}
      boxExtents=<20 10 10>
      constraintSetup {angularLimitsMax=<1 10 15>, angularLimitsMin=<-1 -30 -15>, angularTarget=<0 0 0>}
      localJointOffset=<2 0 0>
     }]
    relativeSpaceBone {boneName=dyn_root_simspace}
   }
   n3 = Anim Dynamics - Bone: dyn_coat_extra_03(ComponentPose=@n2, Alpha=1, ExternalForce="0,0,0") {
    angularDampingOverride=0.9999
    angularSpringConstant=100
    bAngularSpring=true
    bGravityOverrideInSimSpace=true
    bOverrideAngularDamping=true
    simulationSpace=RootRelative
    alphaBoolBlend {blendOption=Linear}
    boundBone {boneName=dyn_coat_extra_03}
    componentAppliedLinearAccClamp=<3500 3500 3500>
    componentLinearAccScale=<1.2 1.2 1.2>
    componentLinearVelScale=<1.2 1.2 1.2>
    physicsBodyDefinitions [{
      sphereCollisionRadius=0
      boundBone {boneName=dyn_coat_extra_03}
      boxExtents=<20 10 10>
      constraintSetup {angularLimitsMax=<1 10 15>, angularLimitsMin=<-1 -30 -15>, angularTarget=<0 0 0>}
      localJointOffset=<2 0 0>
     }]
    relativeSpaceBone {boneName=dyn_root_simspace}
   }
   n4 = Anim Dynamics - Bone: dyn_coat_extra_04(ComponentPose=@n3, Alpha=1, ExternalForce="0,0,0") {
    angularDampingOverride=0.9999
    angularSpringConstant=100
    bAngularSpring=true
    bGravityOverrideInSimSpace=true
    bOverrideAngularDamping=true
    simulationSpace=RootRelative
    alphaBoolBlend {blendOption=Linear}
    boundBone {boneName=dyn_coat_extra_04}
    componentAppliedLinearAccClamp=<3500 3500 3500>
    componentLinearAccScale=<1.2 1.2 1.2>
    componentLinearVelScale=<1.2 1.2 1.2>
    physicsBodyDefinitions [{
      sphereCollisionRadius=0
      boundBone {boneName=dyn_coat_extra_04}
      boxExtents=<20 10 10>
      constraintSetup {angularLimitsMax=<1 10 15>, angularLimitsMin=<-1 -30 -15>, angularTarget=<0 0 0>}
      localJointOffset=<2 0 0>
     }]
    relativeSpaceBone {boneName=dyn_root_simspace}
   }
   n5 = Anim Dynamics - Bone: dyn_coat_extra_05(ComponentPose=@n4, Alpha=1, ExternalForce="0,0,0") {
    angularDampingOverride=0.9999
    angularSpringConstant=100
    bAngularSpring=true
    bGravityOverrideInSimSpace=true
    bOverrideAngularDamping=true
    simulationSpace=RootRelative
    alphaBoolBlend {blendOption=Linear}
    boundBone {boneName=dyn_coat_extra_05}
    componentAppliedLinearAccClamp=<3500 3500 3500>
    componentLinearAccScale=<1.2 1.2 1.2>
    componentLinearVelScale=<1.2 1.2 1.2>
    physicsBodyDefinitions [{
      sphereCollisionRadius=0
      boundBone {boneName=dyn_coat_extra_05}
      boxExtents=<20 10 10>
      constraintSetup {angularLimitsMax=<1 10 15>, angularLimitsMin=<-1 -30 -15>, angularTarget=<0 0 0>}
      localJointOffset=<2 0 0>
     }]
    relativeSpaceBone {boneName=dyn_root_simspace}
   }
   n6 = Anim Dynamics - Bone: dyn_coat_extra_06(ComponentPose=@n5, Alpha=1, ExternalForce="0,0,0") {
    angularDampingOverride=0.9999
    angularSpringConstant=100
    bAngularSpring=true
    bGravityOverrideInSimSpace=true
    bOverrideAngularDamping=true
    simulationSpace=RootRelative
    alphaBoolBlend {blendOption=Linear}
    boundBone {boneName=dyn_coat_extra_06}
    componentAppliedLinearAccClamp=<3500 3500 3500>
    componentLinearAccScale=<1.2 1.2 1.2>
    componentLinearVelScale=<1.2 1.2 1.2>
    physicsBodyDefinitions [{
      sphereCollisionRadius=0
      boundBone {boneName=dyn_coat_extra_06}
      boxExtents=<20 10 10>
      constraintSetup {angularLimitsMax=<1 10 15>, angularLimitsMin=<-1 -30 -15>, angularTarget=<0 0 0>}
      localJointOffset=<2 0 0>
     }]
    relativeSpaceBone {boneName=dyn_root_simspace}
   }
   n7 = Anim Dynamics - Bone: dyn_coat_extra_07(ComponentPose=@n6, Alpha=1, ExternalForce="0,0,0") {
    angularDampingOverride=0.9999
    angularSpringConstant=100
    bAngularSpring=true
    bGravityOverrideInSimSpace=true
    bOverrideAngularDamping=true
    simulationSpace=RootRelative
    alphaBoolBlend {blendOption=Linear}
    boundBone {boneName=dyn_coat_extra_07}
    componentAppliedLinearAccClamp=<3500 3500 3500>
    componentLinearAccScale=<1.2 1.2 1.2>
    componentLinearVelScale=<1.2 1.2 1.2>
    physicsBodyDefinitions [{
      sphereCollisionRadius=0
      boundBone {boneName=dyn_coat_extra_07}
      boxExtents=<20 10 10>
      constraintSetup {angularLimitsMax=<1 10 15>, angularLimitsMin=<-1 -30 -15>, angularTarget=<0 0 0>}
      localJointOffset=<2 0 0>
     }]
    relativeSpaceBone {boneName=dyn_root_simspace}
   }
   n8 = Anim Dynamics - Bone: dyn_coat_extra_08(ComponentPose=@n7, Alpha=1, ExternalForce="0,0,0") {
    angularDampingOverride=0.9999
    angularSpringConstant=100
    bAngularSpring=true
    bGravityOverrideInSimSpace=true
    bOverrideAngularDamping=true
    simulationSpace=RootRelative
    alphaBoolBlend {blendOption=Linear}
    boundBone {boneName=dyn_coat_extra_08}
    componentAppliedLinearAccClamp=<3500 3500 3500>
    componentLinearAccScale=<1.2 1.2 1.2>
    componentLinearVelScale=<1.2 1.2 1.2>
    physicsBodyDefinitions [{
      sphereCollisionRadius=0
      boundBone {boneName=dyn_coat_extra_08}
      boxExtents=<20 10 10>
      constraintSetup {angularLimitsMax=<1 10 15>, angularLimitsMin=<-1 -30 -15>, angularTarget=<0 0 0>}
      localJointOffset=<2 0 0>
     }]
    relativeSpaceBone {boneName=dyn_root_simspace}
   }
   n9 = Anim Dynamics - Bone: dyn_coat_extra_09(ComponentPose=@n8, Alpha=1, ExternalForce="0,0,0") {
    angularDampingOverride=0.9999
    angularSpringConstant=100
    bAngularSpring=true
    bGravityOverrideInSimSpace=true
    bOverrideAngularDamping=true
    simulationSpace=RootRelative
    alphaBoolBlend {blendOption=Linear}
    boundBone {boneName=dyn_coat_extra_09}
    componentAppliedLinearAccClamp=<3500 3500 3500>
    componentLinearAccScale=<1.2 1.2 1.2>
    componentLinearVelScale=<1.2 1.2 1.2>
    physicsBodyDefinitions [{
      sphereCollisionRadius=0
      boundBone {boneName=dyn_coat_extra_09}
      boxExtents=<20 10 10>
      constraintSetup {angularLimitsMax=<1 10 15>, angularLimitsMin=<-1 -30 -15>, angularTarget=<0 0 0>}
      localJointOffset=<2 0 0>
     }]
    relativeSpaceBone {boneName=dyn_root_simspace}
   }
   n10 = Anim Dynamics - Bone: dyn_coat_extra_010(ComponentPose=@n9, Alpha=1, ExternalForce="0,0,0") {
    angularDampingOverride=0.9999
    angularSpringConstant=100
    bAngularSpring=true
    bGravityOverrideInSimSpace=true
    bOverrideAngularDamping=true
    simulationSpace=RootRelative
    alphaBoolBlend {blendOption=Linear}
    boundBone {boneName=dyn_coat_extra_010}
    componentAppliedLinearAccClamp=<3500 3500 3500>
    componentLinearAccScale=<1.2 1.2 1.2>
    componentLinearVelScale=<1.2 1.2 1.2>
    physicsBodyDefinitions [{
      sphereCollisionRadius=0
      boundBone {boneName=dyn_coat_extra_010}
      boxExtents=<20 10 10>
      constraintSetup {angularLimitsMax=<1 10 15>, angularLimitsMin=<-1 -30 -15>, angularTarget=<0 0 0>}
      localJointOffset=<2 0 0>
     }]
    relativeSpaceBone {boneName=dyn_root_simspace}
   }
   n11 = Anim Dynamics - Bone: dyn_coat_extra_011(ComponentPose=@n10, Alpha=1, ExternalForce="0,0,0") {
    angularDampingOverride=0.9999
    angularSpringConstant=100
    bAngularSpring=true
    bGravityOverrideInSimSpace=true
    bOverrideAngularDamping=true
    simulationSpace=RootRelative
    alphaBoolBlend {blendOption=Linear}
    boundBone {boneName=dyn_coat_extra_011}
    componentAppliedLinearAccClamp=<3500 3500 3500>
    componentLinearAccScale=<1.2 1.2 1.2>
    componentLinearVelScale=<1.2 1.2 1.2>
    physicsBodyDefinitions [{
      sphereCollisionRadius=0
      boundBone {boneName=dyn_coat_extra_011}
      boxExtents=<20 10 10>
      constraintSetup {angularLimitsMax=<1 10 15>, angularLimitsMin=<-1 -30 -15>, angularTarget=<0 0 0>}
      localJointOffset=<2 0 0>
     }]
    relativeSpaceBone {boneName=dyn_root_simspace}
   }
   n12 = Anim Dynamics - Bone: dyn_coat_extra_012(ComponentPose=@n11, Alpha=1, ExternalForce="0,0,0") {
    angularDampingOverride=0.9999
    angularSpringConstant=100
    bAngularSpring=true
    bGravityOverrideInSimSpace=true
    bOverrideAngularDamping=true
    simulationSpace=RootRelative
    alphaBoolBlend {blendOption=Linear}
    boundBone {boneName=dyn_coat_extra_012}
    componentAppliedLinearAccClamp=<3500 3500 3500>
    componentLinearAccScale=<1.2 1.2 1.2>
    componentLinearVelScale=<1.2 1.2 1.2>
    physicsBodyDefinitions [{
      sphereCollisionRadius=0
      boundBone {boneName=dyn_coat_extra_012}
      boxExtents=<20 10 10>
      constraintSetup {angularLimitsMax=<1 10 15>, angularLimitsMin=<-1 -30 -15>, angularTarget=<0 0 0>}
      localJointOffset=<2 0 0>
     }]
    relativeSpaceBone {boneName=dyn_root_simspace}
   }
   n13 = Anim Dynamics - Bone: dyn_coat_extra_013(ComponentPose=@n12, Alpha=1, ExternalForce="0,0,0") {
    angularDampingOverride=0.9999
    angularSpringConstant=100
    bAngularSpring=true
    bGravityOverrideInSimSpace=true
    bOverrideAngularDamping=true
    simulationSpace=RootRelative
    alphaBoolBlend {blendOption=Linear}
    boundBone {boneName=dyn_coat_extra_013}
    componentAppliedLinearAccClamp=<3500 3500 3500>
    componentLinearAccScale=<1.2 1.2 1.2>
    componentLinearVelScale=<1.2 1.2 1.2>
    physicsBodyDefinitions [{
      sphereCollisionRadius=0
      boundBone {boneName=dyn_coat_extra_013}
      boxExtents=<20 10 10>
      constraintSetup {angularLimitsMax=<1 10 15>, angularLimitsMin=<-1 -30 -15>, angularTarget=<0 0 0>}
      localJointOffset=<2 0 0>
     }]
    relativeSpaceBone {boneName=dyn_root_simspace}
   }
   n14 = Anim Dynamics - Bone: dyn_coat_extra_014(ComponentPose=@n13, Alpha=1, ExternalForce="0,0,0") {
    angularDampingOverride=0.9999
    angularSpringConstant=100
    bAngularSpring=true
    bGravityOverrideInSimSpace=true
    bOverrideAngularDamping=true
    simulationSpace=RootRelative
    alphaBoolBlend {blendOption=Linear}
    boundBone {boneName=dyn_coat_extra_014}
    componentAppliedLinearAccClamp=<3500 3500 3500>
    componentLinearAccScale=<1.2 1.2 1.2>
    componentLinearVelScale=<1.2 1.2 1.2>
    physicsBodyDefinitions [{
      sphereCollisionRadius=0
      boundBone {boneName=dyn_coat_extra_014}
      boxExtents=<20 10 10>
      constraintSetup {angularLimitsMax=<1 10 15>, angularLimitsMin=<-1 -30 -15>, angularTarget=<0 0 0>}
      localJointOffset=<2 0 0>
     }]
    relativeSpaceBone {boneName=dyn_root_simspace}
   }
   n15 = Anim Dynamics - Bone: dyn_coat_extra_015(ComponentPose=@n14, Alpha=1, ExternalForce="0,0,0") {
    angularDampingOverride=0.9999
    angularSpringConstant=100
    bAngularSpring=true
    bGravityOverrideInSimSpace=true
    bOverrideAngularDamping=true
    simulationSpace=RootRelative
    alphaBoolBlend {blendOption=Linear}
    boundBone {boneName=dyn_coat_extra_015}
    componentAppliedLinearAccClamp=<3500 3500 3500>
    componentLinearAccScale=<1.2 1.2 1.2>
    componentLinearVelScale=<1.2 1.2 1.2>
    physicsBodyDefinitions [{
      sphereCollisionRadius=0
      boundBone {boneName=dyn_coat_extra_015}
      boxExtents=<20 10 10>
      constraintSetup {angularLimitsMax=<1 10 15>, angularLimitsMin=<-1 -30 -15>, angularTarget=<0 0 0>}
      localJointOffset=<2 0 0>
     }]
    relativeSpaceBone {boneName=dyn_root_simspace}
   }
   n16 = Anim Dynamics - Bone: dyn_coat_extra_016(ComponentPose=@n15, Alpha=1, ExternalForce="0,0,0") {
    angularDampingOverride=0.9999
    angularSpringConstant=100
    bAngularSpring=true
    bGravityOverrideInSimSpace=true
    bOverrideAngularDamping=true
    simulationSpace=RootRelative
    alphaBoolBlend {blendOption=Linear}
    boundBone {boneName=dyn_coat_extra_016}
    componentAppliedLinearAccClamp=<3500 3500 3500>
    componentLinearAccScale=<1.2 1.2 1.2>
    componentLinearVelScale=<1.2 1.2 1.2>
    physicsBodyDefinitions [{
      sphereCollisionRadius=0
      boundBone {boneName=dyn_coat_extra_016}
      boxExtents=<20 10 10>
      constraintSetup {angularLimitsMax=<1 10 15>, angularLimitsMin=<-1 -30 -15>, angularTarget=<0 0 0>}
      localJointOffset=<2 0 0>
     }]
    relativeSpaceBone {boneName=dyn_root_simspace}
   }
   n17 = Anim Dynamics - Bone: dyn_coat_extra_017(ComponentPose=@n16, Alpha=1, ExternalForce="0,0,0") {
    angularDampingOverride=0.9999
    angularSpringConstant=100
    bAngularSpring=true
    bGravityOverrideInSimSpace=true
    bOverrideAngularDamping=true
    simulationSpace=RootRelative
    alphaBoolBlend {blendOption=Linear}
    boundBone {boneName=dyn_coat_extra_017}
    componentAppliedLinearAccClamp=<3500 3500 3500>
    componentLinearAccScale=<1.2 1.2 1.2>
    componentLinearVelScale=<1.2 1.2 1.2>
    physicsBodyDefinitions [{
      sphereCollisionRadius=0
      boundBone {boneName=dyn_coat_extra_017}
      boxExtents=<20 10 10>
      constraintSetup {angularLimitsMax=<1 10 15>, angularLimitsMin=<-1 -30 -15>, angularTarget=<0 0 0>}
      localJointOffset=<2 0 0>
     }]
    relativeSpaceBone {boneName=dyn_root_simspace}
   }
   n18 = Anim Dynamics - Bone: dyn_coat_extra_018(ComponentPose=@n17, Alpha=1, ExternalForce="0,0,0") {
    angularDampingOverride=0.9999
    angularSpringConstant=100
    bAngularSpring=true
    bGravityOverrideInSimSpace=true
    bOverrideAngularDamping=true
    simulationSpace=RootRelative
    alphaBoolBlend {blendOption=Linear}
    boundBone {boneName=dyn_coat_extra_018}
    componentAppliedLinearAccClamp=<3500 3500 3500>
    componentLinearAccScale=<1.2 1.2 1.2>
    componentLinearVelScale=<1.2 1.2 1.2>
    physicsBodyDefinitions [{
      sphereCollisionRadius=0
      boundBone {boneName=dyn_coat_extra_018}
      boxExtents=<20 10 10>
      constraintSetup {angularLimitsMax=<1 10 15>, angularLimitsMin=<-1 -30 -15>, angularTarget=<0 0 0>}
      localJointOffset=<2 0 0>
     }]
    relativeSpaceBone {boneName=dyn_root_simspace}
   }
   n19 = Anim Dynamics - Bone: dyn_coat_extra_019(ComponentPose=@n18, Alpha=1, ExternalForce="0,0,0") {
    angularDampingOverride=0.9999
    angularSpringConstant=100
    bAngularSpring=true
    bGravityOverrideInSimSpace=true
    bOverrideAngularDamping=true
    simulationSpace=RootRelative
    alphaBoolBlend {blendOption=Linear}
    boundBone {boneName=dyn_coat_extra_019}
    componentAppliedLinearAccClamp=<3500 3500 3500>
    componentLinearAccScale=<1.2 1.2 1.2>
    componentLinearVelScale=<1.2 1.2 1.2>
    physicsBodyDefinitions [{
      sphereCollisionRadius=0
      boundBone {boneName=dyn_coat_extra_019}
      boxExtents=<20 10 10>
      constraintSetup {angularLimitsMax=<1 10 15>, angularLimitsMin=<-1 -30 -15>, angularTarget=<0 0 0>}
      localJointOffset=<2 0 0>
     }]
    relativeSpaceBone {boneName=dyn_root_simspace}
   }
   n20 = Anim Dynamics - Bone: dyn_coat_extra_020(ComponentPose=@n19, Alpha=1, ExternalForce="0,0,0") {
    angularDampingOverride=0.9999
    angularSpringConstant=100
    bAngularSpring=true
    bGravityOverrideInSimSpace=true
    bOverrideAngularDamping=true
    simulationSpace=RootRelative
    alphaBoolBlend {blendOption=Linear}
    boundBone {boneName=dyn_coat_extra_020}
    componentAppliedLinearAccClamp=<3500 3500 3500>
    componentLinearAccScale=<1.2 1.2 1.2>
    componentLinearVelScale=<1.2 1.2 1.2>
    physicsBodyDefinitions [{
      sphereCollisionRadius=0
      boundBone {boneName=dyn_coat_extra_020}
      boxExtents=<20 10 10>
      constraintSetup {angularLimitsMax=<1 10 15>, angularLimitsMin=<-1 -30 -15>, angularTarget=<0 0 0>}
      localJointOffset=<2 0 0>
     }]
    relativeSpaceBone {boneName=dyn_root_simspace}
   }
   n21 = Anim Dynamics - Bone: dyn_coat_extra_021(ComponentPose=@n20, Alpha=1, ExternalForce="0,0,0") {
    angularDampingOverride=0.9999
    angularSpringConstant=100
    bAngularSpring=true
    bGravityOverrideInSimSpace=true
    bOverrideAngularDamping=true
    simulationSpace=RootRelative
    alphaBoolBlend {blendOption=Linear}
    boundBone {boneName=dyn_coat_extra_021}
    componentAppliedLinearAccClamp=<3500 3500 3500>
    componentLinearAccScale=<1.2 1.2 1.2>
    componentLinearVelScale=<1.2 1.2 1.2>
    physicsBodyDefinitions [{
      sphereCollisionRadius=0
      boundBone {boneName=dyn_coat_extra_021}
      boxExtents=<20 10 10>
      constraintSetup {angularLimitsMax=<1 10 15>, angularLimitsMin=<-1 -30 -15>, angularTarget=<0 0 0>}
      localJointOffset=<2 0 0>
     }]
    relativeSpaceBone {boneName=dyn_root_simspace}
   }
   Anim Dynamics - Bone: dyn_coat_extra_022(ComponentPose=@n21, Alpha=1, ExternalForce="0,0,0") {
    angularDampingOverride=0.9999
    angularSpringConstant=100
    bAngularSpring=true
    bGravityOverrideInSimSpace=true
    bOverrideAngularDamping=true
    simulationSpace=RootRelative
    alphaBoolBlend {blendOption=Linear}
    boundBone {boneName=dyn_coat_extra_022}
    componentAppliedLinearAccClamp=<3500 3500 3500>
    componentLinearAccScale=<1.2 1.2 1.2>
    componentLinearVelScale=<1.2 1.2 1.2>
    physicsBodyDefinitions [{
      sphereCollisionRadius=0
      boundBone {boneName=dyn_coat_extra_022}
      boxExtents=<20 10 10>
      constraintSetup {angularLimitsMax=<1 10 15>, angularLimitsMin=<-1 -30 -15>, angularTarget=<0 0 0>}
      localJointOffset=<2 0 0>
     }]
    relativeSpaceBone {boneName=dyn_root_simspace}
   }
   "Constraint - dyn_coat_extra_014 (PPPPPPPP)"!(
    ComponentPose
    Alpha=1
    ConstraintWeights_0=0.0056
    ConstraintWeights_1:float=0.0306
    ConstraintWeights_2:float=0.0085
    ConstraintWeights_3:float=0.0052
    ConstraintWeights_4:float=0.8143
    ConstraintWeights_5:float=0.0051
    ConstraintWeights_6:float=0.0685
    ConstraintWeights_7:float=0.0623
   ) -> (Pose) {
    alphaBoolBlend {blendOption=Linear}
    boneToModify {boneName=dyn_coat_extra_014}
    constraintSetup [
     {targetBone {boneName=dyn_jacketSide_02_r}}
     {targetBone {boneName=dyn_jacketFront_04_r}}
     {targetBone {boneName=dyn_jacketBack_03_r}}
     {targetBone {boneName=dyn_jacketFront_03_r}}
     {targetBone {boneName=dyn_jacketSide_04_r}}
     {targetBone {boneName=dyn_jacketFront_02_r}}
     {targetBone {boneName=dyn_jacketSide_03_r}}
     {targetBone {boneName=dyn_jacketBack_04_r}}
    ]
    constraintWeights [1,1,1,1,1,1,1,1]
   }
   "Constraint - dyn_coat_extra_015 (PPPPPPPPP)"!(
    ComponentPose
    Alpha=1
    ConstraintWeights_0=0.0701
    ConstraintWeights_1:float=0.0737
    ConstraintWeights_2:float=0.0117
    ConstraintWeights_3:float=0.0023
    ConstraintWeights_4:float=0.3553
    ConstraintWeights_5:float=0.1693
    ConstraintWeights_6:float=0.1039
    ConstraintWeights_7:float=0.0396
    ConstraintWeights_8:float=0.1741
   ) -> (Pose) {
    alphaBoolBlend {blendOption=Linear}
    boneToModify {boneName=dyn_coat_extra_015}
    constraintSetup [
     {targetBone {boneName=dyn_jacketFront_04_r}}
     {targetBone {boneName=dyn_jacketSide_05_r}}
     {targetBone {boneName=dyn_jacketBack_03_r}}
     {targetBone {boneName=dyn_jacketFront_03_r}}
     {targetBone {boneName=dyn_jacketSide_04_r}}
     {targetBone {boneName=dyn_jacketBack_05_r}}
     {targetBone {boneName=dyn_jacketFront_05_r}}
     {targetBone {boneName=dyn_jacketSide_03_r}}
     {targetBone {boneName=dyn_jacketBack_04_r}}
    ]
    constraintWeights [1,1,1,1,1,1,1,1,1]
   }
   "Constraint - dyn_coat_extra_016 (PPPPPPPP)"!(
    ComponentPose
    Alpha=1
    ConstraintWeights_0=0.1225
    ConstraintWeights_1:float=0.0804
    ConstraintWeights_2:float=0.0112
    ConstraintWeights_3:float=0.4658
    ConstraintWeights_4:float=0.0644
    ConstraintWeights_5:float=0.1293
    ConstraintWeights_6:float=0.042
    ConstraintWeights_7:float=0.0844
   ) -> (Pose) {
    alphaBoolBlend {blendOption=Linear}
    boneToModify {boneName=dyn_coat_extra_016}
    constraintSetup [
     {targetBone {boneName=dyn_jacketFront_04_r}}
     {targetBone {boneName=dyn_jacketSide_05_r}}
     {targetBone {boneName=dyn_jacketFront_03_r}}
     {targetBone {boneName=dyn_jacketSide_04_r}}
     {targetBone {boneName=dyn_jacketBack_05_r}}
     {targetBone {boneName=dyn_jacketFront_05_r}}
     {targetBone {boneName=dyn_jacketSide_03_r}}
     {targetBone {boneName=dyn_jacketBack_04_r}}
    ]
    constraintWeights [1,1,1,1,1,1,1,1]
   }
   n22 = Get IsNotCrouchRunning() -> (IsNotCrouchRunning:float)
   "Transform (Modify) Bone - Bone: dyn_jacketBack_03_l"(Alpha=@n22, Translation="0,0,0", Rotation="25,0,0.000", Scale="1,1,1") {
    rotationMode=BMM_Additive
    rotationSpace=BCS_BoneSpace
    alphaBoolBlend {blendOption=Linear}
    alphaScaleBiasClamp {bInterpResult=true, interpSpeedDecreasing=2, interpSpeedIncreasing=5}
    boneToModify {boneName=dyn_jacketBack_03_l}
    rotation {pitch=25}
   }
   "Transform (Modify) Bone - Bone: floorCOL"!(
    ComponentPose
    bAlphaBoolEnabled:bool=NOT Boolean(A=Get bCanPlayCustomAnimations() -> (bCanPlayCustomAnimations:bool))
    Translation=Get SlopeRootTrans() -> (SlopeRootTrans:Vector)
    Rotation=Get SlopeFootRot() -> (SlopeFootRot:Rotator)
    Scale="1,1,1"
   ) -> (Pose) {
    alphaInputType=Bool
    rotationMode=BMM_Replace
    rotationSpace=BCS_WorldSpace
    translationMode=BMM_Replace
    translationSpace=BCS_WorldSpace
    alphaBoolBlend {blendOption=Linear}
    boneToModify {boneName=floorCOL}
   }
   n23 = "Constraint - dyn_coat_extra_01 (PPPPPP)"!(
    ComponentPose=Local To Component(LocalPose=Use cached pose 'post rigid'())
    Alpha=1
    ConstraintWeights_0=0.0275
    ConstraintWeights_1:float=0.0231
    ConstraintWeights_2:float=0.0096
    ConstraintWeights_3:float=0.4844
    ConstraintWeights_4:float=0.3859
    ConstraintWeights_5:float=0.0696
   ) -> (Pose) {
    alphaBoolBlend {blendOption=Linear}
    boneToModify {boneName=dyn_coat_extra_01}
    constraintSetup [
     {targetBone {boneName=dyn_jacketBack_01_l}}
     {targetBone {boneName=dyn_jacketFront_01_l}}
     {targetBone {boneName=spine_02}}
     {targetBone {boneName=dyn_jacketSide_01_l}}
     {targetBone {boneName=pelvis}}
     {targetBone {boneName=dyn_jacketSide_02_l}}
    ]
    constraintWeights [1,1,1,1,1,1]
   }
   n24 = "Constraint - dyn_coat_extra_02 (PPPPP)"!(
    ComponentPose=@n23
    Alpha=1
    ConstraintWeights_0=0.0543
    ConstraintWeights_1:float=0.0558
    ConstraintWeights_2:float=0.0048
    ConstraintWeights_3:float=0.0058
    ConstraintWeights_4:float=0.8794
   ) -> (Pose) {
    alphaBoolBlend {blendOption=Linear}
    boneToModify {boneName=dyn_coat_extra_02}
    constraintSetup [
     {targetBone {boneName=spine_01}}
     {targetBone {boneName=spine_02}}
     {targetBone {boneName=spine_03}}
     {targetBone {boneName=spine_04}}
     {targetBone {boneName=pelvis}}
    ]
    constraintWeights [1,1,1,1,1]
   }
   n25 = "Constraint - dyn_coat_extra_03 (PPPPPPPP)"!(
    ComponentPose=@n24
    Alpha=1
    ConstraintWeights_0=0.0317
    ConstraintWeights_1:float=0.0421
    ConstraintWeights_2:float=0.0129
    ConstraintWeights_3:float=0.1568
    ConstraintWeights_4:float=0.0227
    ConstraintWeights_5:float=0.0428
    ConstraintWeights_6:float=0.0102
    ConstraintWeights_7:float=0.6809
   ) -> (Pose) {
    alphaBoolBlend {blendOption=Linear}
    boneToModify {boneName=dyn_coat_extra_03}
    constraintSetup [
     {targetBone {boneName=dyn_jacketSide_03_l}}
     {targetBone {boneName=dyn_jacketBack_01_l}}
     {targetBone {boneName=dyn_jacketFront_02_l}}
     {targetBone {boneName=dyn_jacketBack_02_l}}
     {targetBone {boneName=dyn_jacketBack_03_l}}
     {targetBone {boneName=dyn_jacketSide_01_l}}
     {targetBone {boneName=pelvis}}
     {targetBone {boneName=dyn_jacketSide_02_l}}
    ]
    constraintWeights [1,1,1,1,1,1,1,1]
   }
   n26 = "Constraint - dyn_coat_extra_04 (PPPPPPPPP)"!(
    ComponentPose=@n25
    Alpha=1
    ConstraintWeights_0=0.055
    ConstraintWeights_1:float=0.021
    ConstraintWeights_2:float=0.0833
    ConstraintWeights_3:float=0.0131
    ConstraintWeights_4:float=0.7226
    ConstraintWeights_5:float=0.0043
    ConstraintWeights_6:float=0.0105
    ConstraintWeights_7:float=0.0399
    ConstraintWeights_8:float=0.0504
   ) -> (Pose) {
    alphaBoolBlend {blendOption=Linear}
    boneToModify {boneName=dyn_coat_extra_04}
    constraintSetup [
     {targetBone {boneName=dyn_jacketBack_04_l}}
     {targetBone {boneName=dyn_jacketFront_03_l}}
     {targetBone {boneName=dyn_jacketSide_03_l}}
     {targetBone {boneName=dyn_jacketFront_02_l}}
     {targetBone {boneName=dyn_jacketSide_04_l}}
     {targetBone {boneName=dyn_jacketBack_03_l}}
     {targetBone {boneName=dyn_jacketSide_05_l}}
     {targetBone {boneName=dyn_jacketSide_02_l}}
     {targetBone {boneName=dyn_jacketFront_04_l}}
    ]
    constraintWeights [1,1,1,1,1,1,1,1,1]
   }
   n27 = "Constraint - dyn_coat_extra_05 (PPPPPPPP)"!(
    ComponentPose=@n26
    Alpha=1
    ConstraintWeights_0=0.0382
    ConstraintWeights_1:float=0.0058
    ConstraintWeights_2:float=0.088
    ConstraintWeights_3:float=0.0087
    ConstraintWeights_4:float=0.06
    ConstraintWeights_5:float=0.1378
    ConstraintWeights_6:float=0.5988
    ConstraintWeights_7:float=0.0626
   ) -> (Pose) {
    alphaBoolBlend {blendOption=Linear}
    boneToModify {boneName=dyn_coat_extra_05}
    constraintSetup [
     {targetBone {boneName=dyn_jacketBack_04_l}}
     {targetBone {boneName=dyn_jacketFront_03_l}}
     {targetBone {boneName=dyn_jacketFront_05_l}}
     {targetBone {boneName=dyn_jacketSide_03_l}}
     {targetBone {boneName=dyn_jacketBack_05_l}}
     {targetBone {boneName=dyn_jacketSide_04_l}}
     {targetBone {boneName=dyn_jacketSide_05_l}}
     {targetBone {boneName=dyn_jacketFront_04_l}}
    ]
    constraintWeights [1,1,1,1,1,1,1,1]
   }
   n28 = "Constraint - dyn_coat_extra_06 (PPPPPPPPPP)"!(
    ComponentPose=@n27
    Alpha=1
    ConstraintWeights_0=0.4572
    ConstraintWeights_1:float=0.0574
    ConstraintWeights_2:float=0.0009
    ConstraintWeights_3:float=0.1221
    ConstraintWeights_4:float=0.017
    ConstraintWeights_5:float=0.0156
    ConstraintWeights_6:float=0.2573
    ConstraintWeights_7:float=0.0116
    ConstraintWeights_8:float=0.0502
    ConstraintWeights_9:float=0.0107
   ) -> (Pose) {
    alphaBoolBlend {blendOption=Linear}
    boneToModify {boneName=dyn_coat_extra_06}
    constraintSetup [
     {targetBone {boneName=dyn_jacketBack_04_l}}
     {targetBone {boneName=dyn_jacketSide_03_l}}
     {targetBone {boneName=dyn_jacketBack_01_l}}
     {targetBone {boneName=dyn_jacketBack_02_l}}
     {targetBone {boneName=dyn_jacketBack_03_r}}
     {targetBone {boneName=dyn_jacketSide_04_l}}
     {targetBone {boneName=dyn_jacketBack_03_l}}
     {targetBone {boneName=dyn_jacketBack_02_r}}
     {targetBone {boneName=dyn_jacketSide_02_l}}
     {targetBone {boneName=dyn_jacketBack_04_r}}
    ]
    constraintWeights [1,1,1,1,1,1,1,1,1,1]
   }
   n29 = "Constraint - dyn_coat_extra_07 (PPPPPPPPPPP)"!(
    ComponentPose=@n28
    Alpha=1
    ConstraintWeights_0=0.0002
    ConstraintWeights_1:float=0.016
    ConstraintWeights_2:float=0.5629
    ConstraintWeights_3:float=0.0276
    ConstraintWeights_4:float=0.0001
    ConstraintWeights_5:float=0.0121
    ConstraintWeights_6:float=0.3488
    ConstraintWeights_7:float=0.0004
    ConstraintWeights_8:float=0.0029
    ConstraintWeights_9:float=0.0009
    ConstraintWeights_10:float=0.0281
   ) -> (Pose) {
    alphaBoolBlend {blendOption=Linear}
    boneToModify {boneName=dyn_coat_extra_07}
    constraintSetup [
     {targetBone {boneName=dyn_jacketSide_02_r}}
     {targetBone {boneName=dyn_jacketBack_01_l}}
     {targetBone {boneName=dyn_jacketBack_02_l}}
     {targetBone {boneName=dyn_jacketBack_03_r}}
     {targetBone {boneName=spine_01}}
     {targetBone {boneName=dyn_jacketBack_03_l}}
     {targetBone {boneName=dyn_jacketBack_02_r}}
     {targetBone {boneName=dyn_jacketSide_01_l}}
     {targetBone {boneName=pelvis}}
     {targetBone {boneName=dyn_jacketSide_02_l}}
     {targetBone {boneName=dyn_jacketBack_01_r}}
    ]
    constraintWeights [1,1,1,1,1,1,1,1,1,1,1]
   }
   n30 = "Constraint - dyn_coat_extra_08 (PPPPPPPP)"!(
    ComponentPose=@n29
    Alpha=1
    ConstraintWeights_0=0.0052
    ConstraintWeights_1:float=0.0147
    ConstraintWeights_2:float=0.0008
    ConstraintWeights_3:float=0.005
    ConstraintWeights_4:float=0.0014
    ConstraintWeights_5:float=0.0784
    ConstraintWeights_6:float=0.0289
    ConstraintWeights_7:float=0.8655
   ) -> (Pose) {
    alphaBoolBlend {blendOption=Linear}
    boneToModify {boneName=dyn_coat_extra_08}
    constraintSetup [
     {targetBone {boneName=dyn_jacketSide_02_r}}
     {targetBone {boneName=dyn_jacketBack_01_l}}
     {targetBone {boneName=dyn_jacketBack_02_l}}
     {targetBone {boneName=dyn_jacketSide_01_r}}
     {targetBone {boneName=spine_01}}
     {targetBone {boneName=dyn_jacketBack_02_r}}
     {targetBone {boneName=pelvis}}
     {targetBone {boneName=dyn_jacketBack_01_r}}
    ]
    constraintWeights [1,1,1,1,1,1,1,1]
   }
   n31 = "Constraint - dyn_coat_extra_09 (PPPPPPPP)"!(
    ComponentPose=@n30
    Alpha=1
    ConstraintWeights_0=0.6263
    ConstraintWeights_1:float=0.2254
    ConstraintWeights_2:float=0.0082
    ConstraintWeights_3:float=0.0016
    ConstraintWeights_4:float=0.0192
    ConstraintWeights_5:float=0.002
    ConstraintWeights_6:float=0.0429
    ConstraintWeights_7:float=0.0745
   ) -> (Pose) {
    alphaBoolBlend {blendOption=Linear}
    boneToModify {boneName=dyn_coat_extra_09}
    constraintSetup [
     {targetBone {boneName=dyn_jacketSide_02_r}}
     {targetBone {boneName=dyn_jacketSide_01_r}}
     {targetBone {boneName=dyn_jacketFront_01_r}}
     {targetBone {boneName=spine_02}}
     {targetBone {boneName=dyn_jacketBack_02_r}}
     {targetBone {boneName=dyn_jacketFront_02_r}}
     {targetBone {boneName=pelvis}}
     {targetBone {boneName=dyn_jacketBack_01_r}}
    ]
    constraintWeights [1,1,1,1,1,1,1,1]
   }
   n32 = "Constraint - dyn_coat_extra_010 (PPPPPPPPPPP)"!(
    ComponentPose=@n31
    Alpha=1
    ConstraintWeights_0=0.1983
    ConstraintWeights_1:float=0.0002
    ConstraintWeights_2:float=0.0002
    ConstraintWeights_3:float=0.0042
    ConstraintWeights_4:float=0
    ConstraintWeights_5:float=0
    ConstraintWeights_6:float=0
    ConstraintWeights_7:float=0.7699
    ConstraintWeights_8:float=0.0001
    ConstraintWeights_9:float=0.0001
    ConstraintWeights_10:float=0.0269
   ) -> (Pose) {
    alphaBoolBlend {blendOption=Linear}
    boneToModify {boneName=dyn_coat_extra_010}
    constraintSetup [
     {targetBone {boneName=dyn_jacketSide_02_r}}
     {targetBone {boneName=dyn_jacketBack_01_l}}
     {targetBone {boneName=dyn_jacketBack_02_l}}
     {targetBone {boneName=dyn_jacketSide_01_r}}
     {targetBone {boneName=dyn_jacketFront_01_r}}
     {targetBone {boneName=spine_01}}
     {targetBone {boneName=spine_02}}
     {targetBone {boneName=dyn_jacketBack_02_r}}
     {targetBone {boneName=dyn_jacketFront_02_r}}
     {targetBone {boneName=pelvis}}
     {targetBone {boneName=dyn_jacketBack_01_r}}
    ]
    constraintWeights [1,1,1,1,1,1,1,1,1,1,1]
   }
   n33 = "Constraint - dyn_coat_extra_011 (PPPPPPPPPP)"!(
    ComponentPose=@n32
    Alpha=1
    ConstraintWeights_0=0.0046
    ConstraintWeights_1:float=0.0029
    ConstraintWeights_2:float=0.0518
    ConstraintWeights_3:float=0.0264
    ConstraintWeights_4:float=0.0004
    ConstraintWeights_5:float=0.0138
    ConstraintWeights_6:float=0.0322
    ConstraintWeights_7:float=0.0002
    ConstraintWeights_8:float=0.0191
    ConstraintWeights_9:float=0.8487
   ) -> (Pose) {
    alphaBoolBlend {blendOption=Linear}
    boneToModify {boneName=dyn_coat_extra_011}
    constraintSetup [
     {targetBone {boneName=dyn_jacketSide_02_r}}
     {targetBone {boneName=dyn_jacketSide_05_r}}
     {targetBone {boneName=dyn_jacketBack_03_r}}
     {targetBone {boneName=dyn_jacketSide_04_r}}
     {targetBone {boneName=dyn_jacketBack_03_l}}
     {targetBone {boneName=dyn_jacketBack_02_r}}
     {targetBone {boneName=dyn_jacketBack_05_r}}
     {targetBone {boneName=dyn_jacketBack_01_r}}
     {targetBone {boneName=dyn_jacketSide_03_r}}
     {targetBone {boneName=dyn_jacketBack_04_r}}
    ]
    constraintWeights [1,1,1,1,1,1,1,1,1,1]
   }
   n34 = "Constraint - dyn_coat_extra_012 (PPPPPPP)"!(
    ComponentPose=@n33
    Alpha=1
    ConstraintWeights_0=0.0003
    ConstraintWeights_1:float=0.0015
    ConstraintWeights_2:float=0.0021
    ConstraintWeights_3:float=0.017
    ConstraintWeights_4:float=0.0132
    ConstraintWeights_5:float=0.4353
    ConstraintWeights_6:float=0.5308
   ) -> (Pose) {
    alphaBoolBlend {blendOption=Linear}
    boneToModify {boneName=dyn_coat_extra_012}
    constraintSetup [
     {targetBone {boneName=dyn_jacketSide_02_r}}
     {targetBone {boneName=dyn_jacketBack_01_l}}
     {targetBone {boneName=dyn_jacketSide_01_r}}
     {targetBone {boneName=spine_01}}
     {targetBone {boneName=spine_02}}
     {targetBone {boneName=pelvis}}
     {targetBone {boneName=dyn_jacketBack_01_r}}
    ]
    constraintWeights [1,1,1,1,1,1,1]
   }
   n35 = "Constraint - dyn_coat_extra_013 (PPPPPPPP)"!(
    ComponentPose=@n34
    Alpha=1
    ConstraintWeights_0=0.2272
    ConstraintWeights_1:float=0.0046
    ConstraintWeights_2:float=0.0889
    ConstraintWeights_3:float=0.0124
    ConstraintWeights_4:float=0.0025
    ConstraintWeights_5:float=0.0186
    ConstraintWeights_6:float=0.0577
    ConstraintWeights_7:float=0.588
   ) -> (Pose) {
    alphaBoolBlend {blendOption=Linear}
    boneToModify {boneName=dyn_coat_extra_013}
    constraintSetup [
     {targetBone {boneName=dyn_jacketSide_02_r}}
     {targetBone {boneName=dyn_jacketSide_01_r}}
     {targetBone {boneName=dyn_jacketBack_03_r}}
     {targetBone {boneName=dyn_jacketFront_03_r}}
     {targetBone {boneName=dyn_jacketFront_01_r}}
     {targetBone {boneName=dyn_jacketBack_02_r}}
     {targetBone {boneName=dyn_jacketFront_02_r}}
     {targetBone {boneName=dyn_jacketSide_03_r}}
    ]
    constraintWeights [1,1,1,1,1,1,1,1]
   }
   n36 = "Constraint - dyn_coat_extra_014 (PPPPPPPPPP)"!(
    ComponentPose=@n35
    Alpha=1
    ConstraintWeights_0=0.1249
    ConstraintWeights_1:float=0.0059
    ConstraintWeights_2:float=0.0465
    ConstraintWeights_3:float=0.0304
    ConstraintWeights_4:float=0.0662
    ConstraintWeights_5:float=0.3355
    ConstraintWeights_6:float=0.0151
    ConstraintWeights_7:float=0.0179
    ConstraintWeights_8:float=0.2254
    ConstraintWeights_9:float=0.1322
   ) -> (Pose) {
    alphaBoolBlend {blendOption=Linear}
    boneToModify {boneName=dyn_coat_extra_014}
    constraintSetup [
     {targetBone {boneName=dyn_jacketFront_04_r}}
     {targetBone {boneName=dyn_jacketSide_05_r}}
     {targetBone {boneName=dyn_jacketFront_02_r}}
     {targetBone {boneName=dyn_jacketBack_03_r}}
     {targetBone {boneName=dyn_jacketFront_03_r}}
     {targetBone {boneName=dyn_jacketSide_04_r}}
     {targetBone {boneName=dyn_jacketBack_05_r}}
     {targetBone {boneName=dyn_jacketFront_05_r}}
     {targetBone {boneName=dyn_jacketSide_03_r}}
     {targetBone {boneName=dyn_jacketBack_04_r}}
    ]
    constraintWeights [1,1,1,1,1,1,1,1,1,1]
   }
   n37 = "Constraint - dyn_coat_extra_015 (PPPPPPPP)"!(
    ComponentPose=@n36
    Alpha=1
    ConstraintWeights_0=0.0789
    ConstraintWeights_1:float=0.089
    ConstraintWeights_2:float=0.0172
    ConstraintWeights_3:float=0.2463
    ConstraintWeights_4:float=0.2
    ConstraintWeights_5:float=0.1355
    ConstraintWeights_6:float=0.049
    ConstraintWeights_7:float=0.184
   ) -> (Pose) {
    alphaBoolBlend {blendOption=Linear}
    boneToModify {boneName=dyn_coat_extra_015}
    constraintSetup [
     {targetBone {boneName=dyn_jacketFront_04_r}}
     {targetBone {boneName=dyn_jacketSide_05_r}}
     {targetBone {boneName=dyn_jacketBack_03_r}}
     {targetBone {boneName=dyn_jacketSide_04_r}}
     {targetBone {boneName=dyn_jacketBack_05_r}}
     {targetBone {boneName=dyn_jacketFront_05_r}}
     {targetBone {boneName=dyn_jacketSide_03_r}}
     {targetBone {boneName=dyn_jacketBack_04_r}}
    ]
    constraintWeights [1,1,1,1,1,1,1,1]
   }
   n38 = "Constraint - dyn_coat_extra_016 (PPPPPPPP)"(
    ComponentPose=@n37
    Alpha=1
    ConstraintWeights_0=0.1293
    ConstraintWeights_1=0.1081
    ConstraintWeights_2=0.0202
    ConstraintWeights_3=0.25
    ConstraintWeights_4=0.1328
    ConstraintWeights_5=0.1865
    ConstraintWeights_6=0.0541
    ConstraintWeights_7=0.119
   ) {
    alphaBoolBlend {blendOption=Linear}
    boneToModify {boneName=dyn_coat_extra_016}
    constraintSetup [
     {targetBone {boneName=dyn_jacketFront_04_r}}
     {targetBone {boneName=dyn_jacketSide_05_r}}
     {targetBone {boneName=dyn_jacketFront_03_r}}
     {targetBone {boneName=dyn_jacketSide_04_r}}
     {targetBone {boneName=dyn_jacketBack_05_r}}
     {targetBone {boneName=dyn_jacketFront_05_r}}
     {targetBone {boneName=dyn_jacketSide_03_r}}
     {targetBone {boneName=dyn_jacketBack_04_r}}
    ]
    constraintWeights [1,1,1,1,1,1,1,1]
   }
   n39 = "Constraint - dyn_coat_extra_017 (PPPPPPPPPP)"!(
    ComponentPose=@n38
    Alpha=1
    ConstraintWeights_0=0.1139
    ConstraintWeights_1:float=0.0441
    ConstraintWeights_2:float=0.0022
    ConstraintWeights_3:float=0.0027
    ConstraintWeights_4:float=0.1794
    ConstraintWeights_5:float=0.0172
    ConstraintWeights_6:float=0.0208
    ConstraintWeights_7:float=0.4822
    ConstraintWeights_8:float=0.0021
    ConstraintWeights_9:float=0.1354
   ) -> (Pose) {
    alphaBoolBlend {blendOption=Linear}
    boneToModify {boneName=dyn_coat_extra_017}
    constraintSetup [
     {targetBone {boneName=dyn_jacketSide_02_r}}
     {targetBone {boneName=dyn_jacketFront_04_r}}
     {targetBone {boneName=dyn_jacketSide_01_r}}
     {targetBone {boneName=dyn_jacketBack_03_r}}
     {targetBone {boneName=dyn_jacketFront_03_r}}
     {targetBone {boneName=dyn_jacketFront_01_r}}
     {targetBone {boneName=dyn_jacketSide_04_r}}
     {targetBone {boneName=dyn_jacketFront_02_r}}
     {targetBone {boneName=pelvis}}
     {targetBone {boneName=dyn_jacketSide_03_r}}
    ]
    constraintWeights [1,1,1,1,1,1,1,1,1,1]
   }
   n40 = "Constraint - dyn_coat_extra_018 (PPPPPPP)"!(
    ComponentPose=@n39
    Alpha=1
    ConstraintWeights_0=0.0646
    ConstraintWeights_1:float=0.0936
    ConstraintWeights_2:float=0.0306
    ConstraintWeights_3:float=0.056
    ConstraintWeights_4:float=0.0005
    ConstraintWeights_5:float=0.0002
    ConstraintWeights_6:float=0.7545
   ) -> (Pose) {
    alphaBoolBlend {blendOption=Linear}
    boneToModify {boneName=dyn_coat_extra_018}
    constraintSetup [
     {targetBone {boneName=spine_01}}
     {targetBone {boneName=spine_02}}
     {targetBone {boneName=spine_03}}
     {targetBone {boneName=spine_04}}
     {targetBone {boneName=spine_05}}
     {targetBone {boneName=clavicle_r}}
     {targetBone {boneName=pelvis}}
    ]
    constraintWeights [1,1,1,1,1,1,1]
   }
   n41 = "Constraint - dyn_coat_extra_019 (PPPPPPPP)"!(
    ComponentPose=@n40
    Alpha=1
    ConstraintWeights_0=0.0009
    ConstraintWeights_1:float=0.1285
    ConstraintWeights_2:float=0.1613
    ConstraintWeights_3:float=0.0647
    ConstraintWeights_4:float=0.0469
    ConstraintWeights_5:float=0.0039
    ConstraintWeights_6:float=0.0003
    ConstraintWeights_7:float=0.5935
   ) -> (Pose) {
    alphaBoolBlend {blendOption=Linear}
    boneToModify {boneName=dyn_coat_extra_019}
    constraintSetup [
     {targetBone {boneName=clavicle_l}}
     {targetBone {boneName=spine_01}}
     {targetBone {boneName=spine_02}}
     {targetBone {boneName=spine_03}}
     {targetBone {boneName=spine_04}}
     {targetBone {boneName=spine_05}}
     {targetBone {boneName=clavicle_r}}
     {targetBone {boneName=pelvis}}
    ]
    constraintWeights [1,1,1,1,1,1,1,1]
   }
   n42 = "Constraint - dyn_coat_extra_020 (PPPPPPPP)"!(
    ComponentPose=@n41
    Alpha=1
    ConstraintWeights_0=0.0029
    ConstraintWeights_1:float=0.1395
    ConstraintWeights_2:float=0.1912
    ConstraintWeights_3:float=0.0845
    ConstraintWeights_4:float=0.0851
    ConstraintWeights_5:float=0.014
    ConstraintWeights_6:float=0.0062
    ConstraintWeights_7:float=0.4767
   ) -> (Pose) {
    alphaBoolBlend {blendOption=Linear}
    boneToModify {boneName=dyn_coat_extra_020}
    constraintSetup [
     {targetBone {boneName=clavicle_l}}
     {targetBone {boneName=spine_01}}
     {targetBone {boneName=spine_02}}
     {targetBone {boneName=spine_03}}
     {targetBone {boneName=spine_04}}
     {targetBone {boneName=spine_05}}
     {targetBone {boneName=clavicle_r}}
     {targetBone {boneName=pelvis}}
    ]
    constraintWeights [1,1,1,1,1,1,1,1]
   }
   n43 = "Constraint - dyn_coat_extra_021 (PPPPPPPPP)"!(
    ComponentPose=@n42
    Alpha=1
    ConstraintWeights_0=0.1067
    ConstraintWeights_1:float=0.0385
    ConstraintWeights_2:float=0.1017
    ConstraintWeights_3:float=0.1372
    ConstraintWeights_4:float=0.4131
    ConstraintWeights_5:float=0.1624
    ConstraintWeights_6:float=0.0104
    ConstraintWeights_7:float=0.0268
    ConstraintWeights_8:float=0.0032
   ) -> (Pose) {
    alphaBoolBlend {blendOption=Linear}
    boneToModify {boneName=dyn_coat_extra_021}
    constraintSetup [
     {targetBone {boneName=clavicle_l}}
     {targetBone {boneName=spine_01}}
     {targetBone {boneName=spine_02}}
     {targetBone {boneName=spine_03}}
     {targetBone {boneName=spine_04}}
     {targetBone {boneName=spine_05}}
     {targetBone {boneName=upperarm_twist_01_l}}
     {targetBone {boneName=clavicle_r}}
     {targetBone {boneName=pelvis}}
    ]
    constraintWeights [1,1,1,1,1,1,1,1,1]
   }
   n44 = "Constraint - dyn_coat_extra_022 (PPPPPPPP)"!(
    ComponentPose=@n43
    Alpha=1
    ConstraintWeights_0=0.1241
    ConstraintWeights_1:float=0.0281
    ConstraintWeights_2:float=0.0777
    ConstraintWeights_3:float=0.1077
    ConstraintWeights_4:float=0.3855
    ConstraintWeights_5:float=0.2228
    ConstraintWeights_6:float=0.0103
    ConstraintWeights_7:float=0.0437
   ) -> (Pose) {
    alphaBoolBlend {blendOption=Linear}
    boneToModify {boneName=dyn_coat_extra_022}
    constraintSetup [
     {targetBone {boneName=clavicle_l}}
     {targetBone {boneName=spine_01}}
     {targetBone {boneName=spine_02}}
     {targetBone {boneName=spine_03}}
     {targetBone {boneName=spine_04}}
     {targetBone {boneName=spine_05}}
     {targetBone {boneName=upperarm_twist_01_l}}
     {targetBone {boneName=clavicle_r}}
    ]
    constraintWeights [1,1,1,1,1,1,1,1]
   }
   n45 = "Constraint - dyn_necklace_MSTR (PPPPPPPP)"!(
    ComponentPose=@n44
    Alpha=1
    ConstraintWeights_0=0.0004
    ConstraintWeights_1:float=0.0043
    ConstraintWeights_2:float=0.0012
    ConstraintWeights_3:float=0.0014
    ConstraintWeights_4:float=0.0014
    ConstraintWeights_5:float=0.3012
    ConstraintWeights_6:float=0.6817
    ConstraintWeights_7:float=0.0085
   ) -> (Pose) {
    alphaBoolBlend {blendOption=Linear}
    boneToModify {boneName=dyn_necklace_MSTR}
    constraintSetup [
     {targetBone {boneName=deform_deltoid_fr_r}}
     {targetBone {boneName=clavicle_l}}
     {targetBone {boneName=upperarm_twist_01_r}}
     {targetBone {boneName=deform_pectoral_fr_r}}
     {targetBone {boneName=deform_pectoral_fr_l}}
     {targetBone {boneName=spine_04}}
     {targetBone {boneName=spine_05}}
     {targetBone {boneName=clavicle_r}}
    ]
    constraintWeights [1,1,1,1,1,1,1,1]
   }
   n46 = Anim Dynamics - Bone: dyn_necklace_01(ComponentPose=@n45, Alpha=1, ExternalForce="0,0,0") {
    angularDampingOverride=0.998
    angularSpringConstant=70
    bAngularSpring=true
    bGravityOverrideInSimSpace=true
    bOverrideAngularDamping=true
    bUseGravityOverride=true
    bUsePlanarLimit=false
    bUseSphericalLimits=true
    simulationSpace=RootRelative
    alphaBoolBlend {blendOption=Linear}
    boundBone {boneName=dyn_necklace_01}
    componentAppliedLinearAccClamp=<3000 3000 3000>
    componentLinearAccScale=<0.3 0.3 0.3>
    componentLinearVelScale=<0.3 0.3 0.3>
    gravityOverride {z=-980}
    physicsBodyDefinitions [{
      collisionType=InnerSphere
      sphereCollisionRadius=1
      boundBone {boneName=dyn_necklace_01}
      boxExtents=<7 12 12>
      constraintSetup {twistAxis=AxisZ, angularLimitsMax=<0 0 5>, angularLimitsMin=<0 -35 -5>, angularTarget=<0 0 0>}
      localJointOffset=<2 0 0>
     }]
    relativeSpaceBone {boneName=dyn_root_simspace}
   }
   n47 = Anim Dynamics - Bone: dyn_necklace_02(ComponentPose=@n46, Alpha=1, ExternalForce="0,0,0") {
    angularDampingOverride=0.998
    angularSpringConstant=70
    bAngularSpring=true
    bGravityOverrideInSimSpace=true
    bOverrideAngularDamping=true
    bUseGravityOverride=true
    bUsePlanarLimit=false
    bUseSphericalLimits=true
    simulationSpace=RootRelative
    alphaBoolBlend {blendOption=Linear}
    boundBone {boneName=dyn_necklace_02}
    componentAppliedLinearAccClamp=<3000 3000 3000>
    componentLinearAccScale=<0.3 0.3 0.3>
    componentLinearVelScale=<0.3 0.3 0.3>
    gravityOverride {z=-980}
    physicsBodyDefinitions [{
      collisionType=InnerSphere
      sphereCollisionRadius=1
      boundBone {boneName=dyn_necklace_02}
      boxExtents=<7 12 12>
      constraintSetup {twistAxis=AxisZ, angularLimitsMax=<0 0 5>, angularLimitsMin=<0 -35 -5>, angularTarget=<0 0 0>}
      localJointOffset=<2 0 0>
     }]
    relativeSpaceBone {boneName=dyn_root_simspace}
    sphericalLimits [
     {limitRadius=7, drivingBone {boneName=spine_04}, sphereLocalOffset=<0 3.1 4>}
     {limitRadius=7, drivingBone {boneName=spine_04}, sphereLocalOffset=<0 3.1 -4>}
    ]
   }
   n48 = Anim Dynamics - Bone: dyn_necklace_03(ComponentPose=@n47, Alpha=1, ExternalForce="0,0,0") {
    angularDampingOverride=0.998
    angularSpringConstant=70
    bAngularSpring=true
    bGravityOverrideInSimSpace=true
    bOverrideAngularDamping=true
    bUseGravityOverride=true
    bUsePlanarLimit=false
    bUseSphericalLimits=true
    simulationSpace=RootRelative
    alphaBoolBlend {blendOption=Linear}
    boundBone {boneName=dyn_necklace_03}
    componentAppliedLinearAccClamp=<3000 3000 3000>
    componentLinearAccScale=<0.3 0.3 0.3>
    componentLinearVelScale=<0.3 0.3 0.3>
    gravityOverride {z=-980}
    physicsBodyDefinitions [{
      collisionType=InnerSphere
      sphereCollisionRadius=1
      boundBone {boneName=dyn_necklace_03}
      boxExtents=<7 12 12>
      constraintSetup {twistAxis=AxisZ, angularLimitsMax=<0 0 7>, angularLimitsMin=<0 -35 -7>, angularTarget=<0 0 0>}
      localJointOffset=<2 0 0>
     }]
    relativeSpaceBone {boneName=dyn_root_simspace}
    sphericalLimits [
     {limitRadius=7, drivingBone {boneName=spine_04}, sphereLocalOffset=<0 3 4>}
     {limitRadius=7, drivingBone {boneName=spine_04}, sphereLocalOffset=<0 3 -4>}
    ]
   }
   n49 = Anim Dynamics - Bone: dyn_necklace_04(ComponentPose=@n48, Alpha=1, ExternalForce="0,0,0") {
    angularDampingOverride=0.998
    angularSpringConstant=100
    bAngularSpring=true
    bGravityOverrideInSimSpace=true
    bOverrideAngularDamping=true
    bUseGravityOverride=true
    bUsePlanarLimit=false
    bUseSphericalLimits=true
    simulationSpace=RootRelative
    alphaBoolBlend {blendOption=Linear}
    boundBone {boneName=dyn_necklace_04}
    componentAppliedLinearAccClamp=<3000 3000 3000>
    componentLinearAccScale=<0.3 0.3 0.3>
    componentLinearVelScale=<0.3 0.3 0.3>
    gravityOverride {z=-980}
    physicsBodyDefinitions [{
      collisionType=InnerSphere
      sphereCollisionRadius=1
      boundBone {boneName=dyn_necklace_04}
      boxExtents=<10 5 5>
      constraintSetup {twistAxis=AxisZ, angularLimitsMax=<5 5 5>, angularLimitsMin=<-5 -35 -5>, angularTarget=<0 0 0>}
      localJointOffset=<2 0 0>
     }]
    relativeSpaceBone {boneName=dyn_root_simspace}
    sphericalLimits [
     {limitRadius=8.5, drivingBone {boneName=spine_04}, sphereLocalOffset=<0 3 4>}
     {limitRadius=8.5, drivingBone {boneName=spine_04}, sphereLocalOffset=<0 3 -4>}
    ]
   }
   Output Pose(Result=Component To Local(ComponentPose=@n49)) {layerGroup=None}
   n50 = Control Rig(Source=Copy Pose From Mesh() {bUseAttachedParent=true}) {
    alphaBoolBlend {blendOption=Linear}
    controlRigAssetReference {blueprintRigClass=&/F_MED_Werewolf_Deform_CtrlRig_twists_2.F_MED_Werewolf_Deform_CtrlRig_twists_2_C}
   }
   n51 = Control Rig!(
    Source=@n50
    Alpha
    KneeRot_fr_l_114deg:Vector
    KneeTrans_fr_l_114deg:Vector="0.258,-3.591,0.009"
    KneeScale_fr_l_114deg:Vector="1,1,1"
    KneeScale_fr_l_130deg:Vector="1,1,1"
    KneeRot_fr_l_130deg:Vector="0,0,-51.2"
    KneeTrans_fr_l_130deg:Vector="0.258,-3.591,0.009"
   ) -> (Pose) {
    alphaBoolBlend {blendOption=Linear}
    controlRigAssetReference {blueprintRigClass=&/F_MED_Werewolf_Deform_CtrlRig_sculpters_2.F_MED_Werewolf_Deform_CtrlRig_sculpters_2_C}
   }
   n52 = "Transform (Modify) Bone - Bone: dyn_glute_l"(
    ComponentPose=Local To Component(LocalPose=@n51)
    Alpha=Get ThighRotL() -> (ThighRotL:float)
    Translation="-4.689,5.014,-4.028"
    Rotation="0.000,0.000,-55"
    Scale="1.25,1.25,1.25"
   ) {
    rotationMode=BMM_Additive
    rotationSpace=BCS_BoneSpace
    scaleMode=BMM_Additive
    scaleSpace=BCS_BoneSpace
    translationMode=BMM_Additive
    translationSpace=BCS_BoneSpace
    alphaBoolBlend {blendOption=Linear}
    boneToModify {boneName=dyn_glute_l}
    rotation {roll=-55}
    scale=<1.25 1.25 1.25>
    translation=<-4.689 5.014 -4.028>
   }
   n53 = "Transform (Modify) Bone - Bone: dyn_glute_r"(ComponentPose=@n52, Alpha=Get ThighRotR() -> (ThighRotR:float), Translation="-3.75,0,0", Rotation="0.000,0.000,-55", Scale="1.3,1.3,1.3") {
    rotationMode=BMM_Additive
    rotationSpace=BCS_BoneSpace
    scaleMode=BMM_Additive
    scaleSpace=BCS_BoneSpace
    translationMode=BMM_Additive
    translationSpace=BCS_BoneSpace
    alphaBoolBlend {blendOption=Linear}
    boneToModify {boneName=dyn_glute_r}
    rotation {roll=-55}
    scale=<1.3 1.3 1.3>
    translation {x=-3.75}
   }
   n54 = "Transform (Modify) Bone - Bone: dyn_jacketFront_mstr_l"(ComponentPose=@n53, Alpha=Get ThighRotL(), Translation="-2,-0.000,5.531", Rotation="0,0,0", Scale="1,1,1") {
    rotationMode=BMM_Additive
    rotationSpace=BCS_BoneSpace
    scaleMode=BMM_Additive
    scaleSpace=BCS_BoneSpace
    translationMode=BMM_Additive
    translationSpace=BCS_BoneSpace
    alphaBoolBlend {blendOption=Linear}
    boneToModify {boneName=dyn_jacketFront_mstr_l}
    translation {x=-2, z=5.531}
   }
   n55 = "Transform (Modify) Bone - Bone: dyn_jacketFront_mstr_r"(ComponentPose=@n54, Alpha=Get ThighRotR(), Translation="-2,-0.000,6", Rotation="10,5,0", Scale="1,1,1") {
    rotationMode=BMM_Additive
    rotationSpace=BCS_BoneSpace
    scaleMode=BMM_Additive
    scaleSpace=BCS_BoneSpace
    translationMode=BMM_Additive
    translationSpace=BCS_BoneSpace
    alphaBoolBlend {blendOption=Linear}
    boneToModify {boneName=dyn_jacketFront_mstr_r}
    rotation {pitch=10, yaw=5}
    translation {x=-2, z=6}
   }
   n56 = "Transform (Modify) Bone - Bone: floorCOL"(ComponentPose=@n55, bAlphaBoolEnabled=Get bIsFalling(), Translation="0,0,-300", Rotation="0,0,0", Scale="1,1,1") {
    alphaInputType=Bool
    rotationMode=BMM_Additive
    rotationSpace=BCS_BoneSpace
    scaleMode=BMM_Additive
    scaleSpace=BCS_BoneSpace
    translationMode=BMM_Additive
    translationSpace=BCS_BoneSpace
    alphaBoolBlend {blendInTime=0.1, blendOption=Linear, blendOutTime=0.1}
    boneToModify {boneName=floorCOL}
    translation {z=-300}
   }
   n57 = Get bIsSliding()
   n58 = "Transform (Modify) Bone - Bone: dyn_jacketBack_mstr_r"!(ComponentPose=@n56, bAlphaBoolEnabled:bool=@n57, Translation="0,0,5", Rotation="45,0,0", Scale="1,1,1") -> (Pose) {
    alphaInputType=Bool
    rotationMode=BMM_Additive
    rotationSpace=BCS_BoneSpace
    translationMode=BMM_Additive
    alphaBoolBlend {blendInTime=0.1, blendOption=Linear, blendOutTime=0.1}
    boneToModify {boneName=dyn_jacketBack_mstr_r}
    rotation {pitch=45}
    translation {z=5}
   }
   n59 = "Transform (Modify) Bone - Bone: dyn_jacketBack_mstr_l"!(ComponentPose=@n58, bAlphaBoolEnabled:bool=@n57, Translation="0,0,5", Rotation="45,0,0", Scale="1,1,1") -> (Pose) {
    alphaInputType=Bool
    rotationMode=BMM_Additive
    rotationSpace=BCS_BoneSpace
    translationMode=BMM_Additive
    alphaBoolBlend {blendInTime=0.1, blendOption=Linear, blendOutTime=0.1}
    boneToModify {boneName=dyn_jacketBack_mstr_l}
    rotation {pitch=45}
    translation {z=5}
   }
   n60 = "Transform (Modify) Bone - Bone: dyn_jacketSide_mstr_r"!(ComponentPose=@n59, bAlphaBoolEnabled:bool=@n57, Translation="0,0,3", Rotation="35,0,0", Scale="1,1,1") -> (Pose) {
    alphaInputType=Bool
    rotationMode=BMM_Additive
    rotationSpace=BCS_BoneSpace
    translationMode=BMM_Additive
    alphaBoolBlend {blendInTime=0.1, blendOption=Linear, blendOutTime=0.1}
    boneToModify {boneName=dyn_jacketSide_mstr_r}
    rotation {pitch=35}
    translation {z=3}
   }
   n61 = "Transform (Modify) Bone - Bone: dyn_jacketSide_mstr_l"!(ComponentPose=@n60, bAlphaBoolEnabled:bool=@n57, Translation="0,0,3", Rotation="35,0,0", Scale="1,1,1") -> (Pose) {
    alphaInputType=Bool
    rotationMode=BMM_Additive
    rotationSpace=BCS_BoneSpace
    translationMode=BMM_Additive
    alphaBoolBlend {blendInTime=0.1, blendOption=Linear, blendOutTime=0.1}
    boneToModify {boneName=dyn_jacketSide_mstr_l}
    rotation {pitch=35}
    translation {z=3}
   }
   n62 = "Transform (Modify) Bone - Bone: dyn_jacketFront_mstr_r"!(ComponentPose=@n61, bAlphaBoolEnabled:bool=@n57, Translation="0,0,3", Rotation="45,0,0", Scale="1,1,1") -> (Pose) {
    alphaInputType=Bool
    rotationMode=BMM_Additive
    rotationSpace=BCS_BoneSpace
    translationMode=BMM_Additive
    alphaBoolBlend {blendInTime=0.1, blendOption=Linear, blendOutTime=0.1}
    boneToModify {boneName=dyn_jacketFront_mstr_r}
    rotation {pitch=45}
    translation {z=3}
   }
   n63 = "Transform (Modify) Bone - Bone: dyn_jacketFront_mstr_l"!(ComponentPose=@n62, bAlphaBoolEnabled:bool=@n57, Translation="0,0,3", Rotation="35,0,0", Scale="1,1,1") -> (Pose) {
    alphaInputType=Bool
    rotationMode=BMM_Additive
    rotationSpace=BCS_BoneSpace
    translationMode=BMM_Additive
    alphaBoolBlend {blendInTime=0.1, blendOption=Linear, blendOutTime=0.1}
    boneToModify {boneName=dyn_jacketFront_mstr_l}
    rotation {pitch=35}
    translation {z=3}
   }
   Save cached pose 'prerigid'(Pose=Component To Local(ComponentPose=@n63)) {cacheName=prerigid}
   n64 = "Transform (Modify) Bone - Bone: dyn_jacketBack_mstr_r"!(
    ComponentPose=Local To Component(LocalPose=Use cached pose 'prerigid'())
    Alpha:float=1
    Translation="0,-5,0"
    Rotation="0,0,0"
    Scale="1,1,1"
   ) -> (Pose) {translationMode=BMM_Additive, alphaBoolBlend {blendOption=Linear}, boneToModify {boneName=dyn_jacketBack_mstr_r}, translation {y=-5}}
   n65 = "Transform (Modify) Bone - Bone: dyn_jacketBack_mstr_l"!(ComponentPose=@n64, Alpha:float=1, Translation="0,-5,0", Rotation="0,0,0", Scale="1,1,1") -> (Pose) {translationMode=BMM_Additive, alphaBoolBlend {blendOption=Linear}, boneToModify {boneName=dyn_jacketBack_mstr_l}, translation {y=-5}}
   n66 = "Transform (Modify) Bone - Bone: dyn_jacketSide_mstr_r"!(ComponentPose=@n65, Alpha:float=1, Translation="-3,-5,0", Rotation="0,0,0", Scale="1,1,1") -> (Pose) {translationMode=BMM_Additive, alphaBoolBlend {blendOption=Linear}, boneToModify {boneName=dyn_jacketSide_mstr_r}, translation=<-3 -5>}
   n67 = "Transform (Modify) Bone - Bone: dyn_jacketSide_mstr_l"!(ComponentPose=@n66, Alpha:float=1, Translation="3,-5,0", Rotation="0,0,0", Scale="1,1,1") -> (Pose) {translationMode=BMM_Additive, alphaBoolBlend {blendOption=Linear}, boneToModify {boneName=dyn_jacketSide_mstr_l}, translation=<3 -5>}
   n68 = RigidBody!(ComponentPose=@n67, Alpha, OverrideWorldGravity:Vector=Get GravityOverride(), ExternalForce) -> (Pose) {
    bClampLinearTranslationLimitToRefPose=true
    bOverrideWorldGravity=true
    overridePhysicsAsset=&/F_MED_Werewolf_perf_skydive_Physics
    simulationSpace=BaseBoneSpace
    alphaBoolBlend {blendOption=Linear}
    baseBoneRef {boneName=dyn_root_simspace}
    componentAppliedLinearAccClamp=<3000 3000 3000>
    componentLinearAccScale=<0.25 0.25 0.4>
    componentLinearVelScale=<0.2 0.2 0.2>
   }
   n69 = Component To Local(ComponentPose=@n68)
   n70 = Get bIsMoving2D()
   n71 = "Transform (Modify) Bone - Bone: dyn_jacketBack_mstr_r"!(
    ComponentPose=Local To Component(LocalPose=Use cached pose 'prerigid'())
    bAlphaBoolEnabled:bool=AND Boolean(A=Get bIsCrouching(), B=NOT Boolean(A=@n70))
    Translation="0,0,0"
    Rotation="45,0,0"
    Scale="1,1,1"
   ) -> (Pose) {
    alphaInputType=Bool
    rotationMode=BMM_Additive
    rotationSpace=BCS_BoneSpace
    translationMode=BMM_Additive
    alphaBoolBlend {blendInTime=0.1, blendOption=Linear, blendOutTime=0.1}
    boneToModify {boneName=dyn_jacketBack_mstr_r}
    rotation {pitch=45}
   }
   n72 = RigidBody(ComponentPose=@n71, OverrideWorldGravity=Get GravityOverride()) {
    bClampLinearTranslationLimitToRefPose=true
    bOverrideWorldGravity=true
    overridePhysicsAsset=&/F_MED_Werewolf_Physics
    simulationSpace=BaseBoneSpace
    alphaBoolBlend {blendOption=Linear}
    baseBoneRef {boneName=dyn_root_simspace}
    componentAppliedLinearAccClamp=<3000 3000 3000>
    componentLinearAccScale=<0.25 0.25 0.4>
    componentLinearVelScale=<0.2 0.2 0.2>
   }
   n73 = Component To Local(ComponentPose=@n72)
   n74 = Get bIsSkyDiving()
   n75 = Get ThighRotInrR() -> (ThighRotInrR:float)
   n76 = Look At - Bone: dyn_jacketFront_mstr_r(ComponentPose=Local To Component(LocalPose=Use cached pose 'prerigid'()), Alpha=@n75) {
    alphaBoolBlend {blendOption=Linear}
    alphaScaleBiasClamp {bInterpResult=true}
    boneToModify {boneName=dyn_jacketFront_mstr_r}
    lookAtLocation=<0 -12 5>
    lookAtTarget {boneReference {boneName=calf_r}}
    lookAt_Axis {axis=<1 0>}
    lookUp_Axis {axis {x=0}}
   }
   n77 = Look At - Bone: dyn_jacketSide_mstr_r(ComponentPose=@n76, Alpha=@n75) {
    alphaBoolBlend {blendOption=Linear}
    alphaScaleBiasClamp {bInterpResult=true}
    boneToModify {boneName=dyn_jacketSide_mstr_r}
    lookAtLocation {x=0, z=11}
    lookAtTarget {boneReference {boneName=calf_r}}
    lookAt_Axis {axis=<1 0>}
    lookUp_Axis {axis {x=0}}
   }
   n78 = Get ThighRotUpR() -> (ThighRotUpR:float)
   n79 = "Transform (Modify) Bone - Bone: dyn_jacketFront_01_r"(ComponentPose=@n77, Alpha=@n78, Translation="0,0,0", Rotation="9.096,4.987,10.829", Scale="1,1,1") {
    rotationMode=BMM_Additive
    rotationSpace=BCS_BoneSpace
    translationMode=BMM_Additive
    translationSpace=BCS_BoneSpace
    alphaBoolBlend {blendOption=Linear}
    boneToModify {boneName=dyn_jacketFront_01_r}
    rotation=<rot 9.096 4.987 10.829>
   }
   n80 = "Transform (Modify) Bone - Bone: dyn_jacketSide_01_r"(ComponentPose=@n79, Alpha=@n78, Translation="0,0,0", Rotation="5.000,0.000,0.000", Scale="1,1,1") {
    rotationMode=BMM_Additive
    rotationSpace=BCS_BoneSpace
    translationMode=BMM_Additive
    translationSpace=BCS_BoneSpace
    alphaBoolBlend {blendOption=Linear}
    boneToModify {boneName=dyn_jacketSide_01_r}
    rotation {pitch=5}
   }
   n81 = Get ThighRotInrL() -> (ThighRotInrL:float)
   n82 = Look At - Bone: dyn_jacketFront_mstr_l(ComponentPose=@n80, Alpha=@n81) {
    alphaBoolBlend {blendOption=Linear}
    alphaScaleBiasClamp {bInterpResult=true}
    boneToModify {boneName=dyn_jacketFront_mstr_l}
    lookAtLocation=<0 12 -5>
    lookAtTarget {boneReference {boneName=calf_l}}
    lookAt_Axis {axis=<1 0>}
    lookUp_Axis {axis {x=0}}
   }
   n83 = Look At - Bone: dyn_jacketSide_mstr_l(ComponentPose=@n82, Alpha=@n81) {
    alphaBoolBlend {blendOption=Linear}
    alphaScaleBiasClamp {bInterpResult=true}
    boneToModify {boneName=dyn_jacketSide_mstr_l}
    lookAtLocation {x=0, z=-11}
    lookAtTarget {boneReference {boneName=calf_l}}
    lookAt_Axis {axis=<1 0>}
    lookUp_Axis {axis {x=0}}
   }
   n84 = Get ThighRotUpL() -> (ThighRotUpL:float)
   n85 = "Transform (Modify) Bone - Bone: dyn_jacketFront_01_l"(ComponentPose=@n83, Alpha=@n84, Translation="0,0,0", Rotation="9.096,-4.987,-10.828", Scale="1,1,1") {
    rotationMode=BMM_Additive
    rotationSpace=BCS_BoneSpace
    translationMode=BMM_Additive
    translationSpace=BCS_BoneSpace
    alphaBoolBlend {blendOption=Linear}
    boneToModify {boneName=dyn_jacketFront_01_l}
    rotation=<rot 9.096 -4.987 -10.8284>
   }
   n86 = "Transform (Modify) Bone - Bone: dyn_jacketSide_01_l"(ComponentPose=@n85, Alpha=@n84, Translation="0,0,0", Rotation="5,0,0", Scale="1,1,1") {
    rotationMode=BMM_Additive
    rotationSpace=BCS_BoneSpace
    translationMode=BMM_Additive
    translationSpace=BCS_BoneSpace
    alphaBoolBlend {blendOption=Linear}
    boneToModify {boneName=dyn_jacketSide_01_l}
    rotation {pitch=5}
   }
   n87 = Get JumpingAlpha() -> (JumpingAlpha:float)
   n88 = "Transform (Modify) Bone - Bone: dyn_jacketBack_01_l"(ComponentPose=@n86, Alpha=@n87, Translation="0,0,0", Rotation="35,10,0.000", Scale="1,1,1") {
    rotationMode=BMM_Additive
    rotationSpace=BCS_BoneSpace
    alphaBoolBlend {blendOption=Linear}
    alphaScaleBiasClamp {bInterpResult=true, interpSpeedDecreasing=2, interpSpeedIncreasing=5}
    boneToModify {boneName=dyn_jacketBack_01_l}
    rotation {pitch=35, yaw=10}
   }
   n89 = "Transform (Modify) Bone - Bone: dyn_jacketBack_01_r"(ComponentPose=@n88, Alpha=@n87, Translation="0,0,0", Rotation="35,-10,0.000", Scale="1,1,1") {
    rotationMode=BMM_Additive
    rotationSpace=BCS_BoneSpace
    alphaBoolBlend {blendOption=Linear}
    alphaScaleBiasClamp {bInterpResult=true, interpSpeedDecreasing=2, interpSpeedIncreasing=5}
    boneToModify {boneName=dyn_jacketBack_01_r}
    rotation {pitch=35, yaw=-10}
   }
   n90 = "Transform (Modify) Bone - Bone: dyn_jacketSide_01_l"(ComponentPose=@n89, Alpha=@n87, Translation="0,0,0", Rotation="15,-15,0.000", Scale="1,1,1") {
    rotationMode=BMM_Additive
    rotationSpace=BCS_BoneSpace
    alphaBoolBlend {blendOption=Linear}
    alphaScaleBiasClamp {bInterpResult=true, interpSpeedDecreasing=2, interpSpeedIncreasing=5}
    boneToModify {boneName=dyn_jacketSide_01_l}
    rotation {pitch=15, yaw=-15}
   }
   n91 = "Transform (Modify) Bone - Bone: dyn_jacketSide_01_l"(ComponentPose=@n90, Alpha=@n87, Translation="0,0,0", Rotation="15,15,0.000", Scale="1,1,1") {
    rotationMode=BMM_Additive
    rotationSpace=BCS_BoneSpace
    alphaBoolBlend {blendOption=Linear}
    alphaScaleBiasClamp {bInterpResult=true, interpSpeedDecreasing=2, interpSpeedIncreasing=5}
    boneToModify {boneName=dyn_jacketSide_01_l}
    rotation {pitch=15, yaw=15}
   }
   n92 = Get SkydiveAlpha() -> (SkydiveAlpha:float)
   n93 = "Transform (Modify) Bone - Bone: dyn_jacketBack_mstr_l"(ComponentPose=@n91, Alpha=@n92, Translation="0,0,0", Rotation="15,0,0.000", Scale="1,1,1") {
    rotationMode=BMM_Additive
    rotationSpace=BCS_BoneSpace
    alphaBoolBlend {blendOption=Linear}
    alphaScaleBiasClamp {bInterpResult=true, interpSpeedDecreasing=2, interpSpeedIncreasing=5}
    boneToModify {boneName=dyn_jacketBack_mstr_l}
    rotation {pitch=15}
   }
   n94 = "Transform (Modify) Bone - Bone: dyn_jacketBack_03_l"(ComponentPose=@n93, Alpha=@n92, Translation="0,0,0", Rotation="20,0,0.000", Scale="1,1,1") {
    rotationMode=BMM_Additive
    rotationSpace=BCS_BoneSpace
    alphaBoolBlend {blendOption=Linear}
    alphaScaleBiasClamp {bInterpResult=true, interpSpeedDecreasing=2, interpSpeedIncreasing=5}
    boneToModify {boneName=dyn_jacketBack_03_l}
    rotation {pitch=20}
   }
   n95 = "Transform (Modify) Bone - Bone: dyn_jacketBack_mstr_r"!(ComponentPose=@n94, Alpha:float=@n92, Translation="0,0,0", Rotation="15,0,0.000", Scale="1,1,1") -> (Pose) {
    rotationMode=BMM_Additive
    rotationSpace=BCS_BoneSpace
    alphaBoolBlend {blendOption=Linear}
    alphaScaleBiasClamp {bInterpResult=true, interpSpeedDecreasing=2, interpSpeedIncreasing=5}
    boneToModify {boneName=dyn_jacketBack_mstr_r}
    rotation {pitch=15}
   }
   n96 = Get ParachuteAlpha() -> (ParachuteAlpha:float)
   n97 = "Transform (Modify) Bone - Bone: dyn_jacketBack_01_l"(ComponentPose=@n95, Alpha=@n96, Translation="0,0,0", Rotation="15,10,0.000", Scale="1,1,1") {
    rotationMode=BMM_Additive
    rotationSpace=BCS_BoneSpace
    alphaBoolBlend {blendOption=Linear}
    alphaScaleBiasClamp {bInterpResult=true, interpSpeedDecreasing=2, interpSpeedIncreasing=5}
    boneToModify {boneName=dyn_jacketBack_01_l}
    rotation {pitch=15, yaw=10}
   }
   n98 = "Transform (Modify) Bone - Bone: dyn_jacketBack_01_r"(ComponentPose=@n97, Alpha=@n96, Translation="0,0,0", Rotation="15,-10,0.000", Scale="1,1,1") {
    rotationMode=BMM_Additive
    rotationSpace=BCS_BoneSpace
    alphaBoolBlend {blendOption=Linear}
    alphaScaleBiasClamp {bInterpResult=true, interpSpeedDecreasing=2, interpSpeedIncreasing=5}
    boneToModify {boneName=dyn_jacketBack_01_r}
    rotation {pitch=15, yaw=-10}
   }
   n99 = "Transform (Modify) Bone - Bone: dyn_jacketSide_01_l"(ComponentPose=@n98, Alpha=@n96, Translation="0,0,0", Rotation="15,-15,0.000", Scale="1,1,1") {
    rotationMode=BMM_Additive
    rotationSpace=BCS_BoneSpace
    alphaBoolBlend {blendOption=Linear}
    alphaScaleBiasClamp {bInterpResult=true, interpSpeedDecreasing=2, interpSpeedIncreasing=5}
    boneToModify {boneName=dyn_jacketSide_01_l}
    rotation {pitch=15, yaw=-15}
   }
   n100 = "Transform (Modify) Bone - Bone: dyn_jacketSide_01_l"(ComponentPose=@n99, Alpha=@n96, Translation="0,0,0", Rotation="15,15,0.000", Scale="1,1,1") {
    rotationMode=BMM_Additive
    rotationSpace=BCS_BoneSpace
    alphaBoolBlend {blendOption=Linear}
    alphaScaleBiasClamp {bInterpResult=true, interpSpeedDecreasing=2, interpSpeedIncreasing=5}
    boneToModify {boneName=dyn_jacketSide_01_l}
    rotation {pitch=15, yaw=15}
   }
   n101 = Component To Local(ComponentPose=@n100)
   n102 = "Transform (Modify) Bone - Bone: dyn_jacketFront_01_r"(ComponentPose=Local To Component(LocalPose=Use cached pose 'prerigid'()), Alpha=1, Translation="0,0,0", Rotation="15,0,0", Scale="1,1,1") {
    rotationMode=BMM_Additive
    rotationSpace=BCS_BoneSpace
    translationMode=BMM_Additive
    translationSpace=BCS_BoneSpace
    alphaBoolBlend {blendOption=Linear}
    boneToModify {boneName=dyn_jacketFront_01_r}
    rotation {pitch=15}
   }
   n103 = "Transform (Modify) Bone - Bone: dyn_jacketSide_01_r"(ComponentPose=@n102, Alpha=1, Translation="0,0,0", Rotation="20,0,0", Scale="1,1,1") {
    rotationMode=BMM_Additive
    rotationSpace=BCS_BoneSpace
    translationMode=BMM_Additive
    translationSpace=BCS_BoneSpace
    alphaBoolBlend {blendOption=Linear}
    boneToModify {boneName=dyn_jacketSide_01_r}
    rotation {pitch=20}
   }
   n104 = "Transform (Modify) Bone - Bone: dyn_jacketBack_01_l"(ComponentPose=@n103, Alpha=@n22, Translation="0,0,0", Rotation="10,10,0.000", Scale="1,1,1") {
    rotationMode=BMM_Additive
    rotationSpace=BCS_BoneSpace
    alphaBoolBlend {blendOption=Linear}
    alphaScaleBiasClamp {bInterpResult=true, interpSpeedDecreasing=5}
    boneToModify {boneName=dyn_jacketBack_01_l}
    rotation {pitch=10, yaw=10}
   }
   n105 = "Transform (Modify) Bone - Bone: dyn_jacketBack_01_r"(ComponentPose=@n104, Alpha=@n22, Translation="0,0,0", Rotation="10,-10,0.000", Scale="1,1,1") {
    rotationMode=BMM_Additive
    rotationSpace=BCS_BoneSpace
    alphaBoolBlend {blendOption=Linear}
    alphaScaleBiasClamp {bInterpResult=true, interpSpeedDecreasing=5}
    boneToModify {boneName=dyn_jacketBack_01_r}
    rotation {pitch=10, yaw=-10}
   }
   n106 = "Transform (Modify) Bone - Bone: dyn_jacketBack_02_r"(ComponentPose=@n105, Alpha=@n22, Translation="0,0,0", Rotation="30,0,0.000", Scale="1,1,1") {
    rotationMode=BMM_Additive
    rotationSpace=BCS_BoneSpace
    alphaBoolBlend {blendOption=Linear}
    alphaScaleBiasClamp {bInterpResult=true, interpSpeedDecreasing=5}
    boneToModify {boneName=dyn_jacketBack_02_r}
    rotation {pitch=30}
   }
   n107 = "Transform (Modify) Bone - Bone: dyn_jacketFront_mstr_r"!(ComponentPose=@n106, Alpha:float=Get ThighRotUpR(), Translation="0,0,3", Rotation="0,0,0", Scale="1,1,1") -> (Pose) {
    rotationMode=BMM_Additive
    rotationSpace=BCS_BoneSpace
    translationMode=BMM_Additive
    translationSpace=BCS_BoneSpace
    alphaBoolBlend {blendOption=Linear}
    boneToModify {boneName=dyn_jacketFront_mstr_r}
    translation {z=3}
   }
   n108 = Get ThighRotOtrR() -> (ThighRotOtrR:float)
   n109 = "Transform (Modify) Bone - Bone: dyn_jacketFront_01_r"(ComponentPose=@n107, Alpha=@n108, Translation="0,0,0", Rotation="45,0,0", Scale="1,1,1") {
    rotationMode=BMM_Additive
    rotationSpace=BCS_BoneSpace
    translationMode=BMM_Additive
    translationSpace=BCS_BoneSpace
    alphaBoolBlend {blendOption=Linear}
    boneToModify {boneName=dyn_jacketFront_01_r}
    rotation {pitch=45}
   }
   n110 = "Transform (Modify) Bone - Bone: dyn_jacketFront_01_r"(ComponentPose=@n109, Alpha=@n108, Translation="0,0,2", Rotation="30,0,0", Scale="1,1,1") {
    rotationMode=BMM_Additive
    rotationSpace=BCS_BoneSpace
    translationMode=BMM_Additive
    translationSpace=BCS_BoneSpace
    alphaBoolBlend {blendOption=Linear}
    boneToModify {boneName=dyn_jacketFront_01_r}
    rotation {pitch=30}
    translation {z=2}
   }
   n111 = "Transform (Modify) Bone - Bone: dyn_jacketSide_02_r"(ComponentPose=@n110, Alpha=@n108, Translation="0,0,0", Rotation="45,0,0", Scale="1,1,1") {
    rotationMode=BMM_Additive
    rotationSpace=BCS_BoneSpace
    translationMode=BMM_Additive
    translationSpace=BCS_BoneSpace
    alphaBoolBlend {blendOption=Linear}
    boneToModify {boneName=dyn_jacketSide_02_r}
    rotation {pitch=45}
   }
   n112 = Get ThighRotUpL()
   n113 = "Transform (Modify) Bone - Bone: dyn_jacketFront_01_l"(ComponentPose=@n111, Alpha=@n112, Translation="0,0,0", Rotation="10,-15.000,-35", Scale="1,1,1") {
    rotationMode=BMM_Additive
    rotationSpace=BCS_BoneSpace
    translationMode=BMM_Additive
    translationSpace=BCS_BoneSpace
    alphaBoolBlend {blendOption=Linear}
    boneToModify {boneName=dyn_jacketFront_01_l}
    rotation=<rot 10 -15 -35>
   }
   n114 = "Transform (Modify) Bone - Bone: dyn_jacketFront_02_l"(ComponentPose=@n113, Alpha=@n112, Translation="0,0,0", Rotation="-25,0,0", Scale="1,1,1") {
    rotationMode=BMM_Additive
    rotationSpace=BCS_BoneSpace
    translationMode=BMM_Additive
    translationSpace=BCS_BoneSpace
    alphaBoolBlend {blendOption=Linear}
    boneToModify {boneName=dyn_jacketFront_02_l}
    rotation {pitch=-25}
   }
   n115 = "Transform (Modify) Bone - Bone: dyn_jacketSide_01_l"(ComponentPose=@n114, Alpha=@n112, Translation="0,-2,3", Rotation="0,-10,0", Scale="1,1,1") {
    rotationMode=BMM_Additive
    rotationSpace=BCS_BoneSpace
    translationMode=BMM_Additive
    translationSpace=BCS_BoneSpace
    alphaBoolBlend {blendOption=Linear}
    boneToModify {boneName=dyn_jacketSide_01_l}
    rotation {yaw=-10}
    translation {y=-2, z=3}
   }
   n116 = Get ThighRotOtrL() -> (ThighRotOtrL:float)
   n117 = "Transform (Modify) Bone - Bone: dyn_jacketFront_01_l"(ComponentPose=@n115, Alpha=@n116, Translation="0,0,0", Rotation="45,0,0", Scale="1,1,1") {
    rotationMode=BMM_Additive
    rotationSpace=BCS_BoneSpace
    translationMode=BMM_Additive
    translationSpace=BCS_BoneSpace
    alphaBoolBlend {blendOption=Linear}
    boneToModify {boneName=dyn_jacketFront_01_l}
    rotation {pitch=45}
   }
   n118 = "Transform (Modify) Bone - Bone: dyn_jacketFront_02_l"(ComponentPose=@n117, Alpha=@n116, Translation="0,0,2", Rotation="30,0,0", Scale="1,1,1") {
    rotationMode=BMM_Additive
    rotationSpace=BCS_BoneSpace
    translationMode=BMM_Additive
    translationSpace=BCS_BoneSpace
    alphaBoolBlend {blendOption=Linear}
    boneToModify {boneName=dyn_jacketFront_02_l}
    rotation {pitch=30}
    translation {z=2}
   }
   n119 = "Transform (Modify) Bone - Bone: dyn_jacketSide_02_l"(ComponentPose=@n118, Alpha=@n116, Translation="0,0,0", Rotation="45,0,0", Scale="1,1,1") {
    rotationMode=BMM_Additive
    rotationSpace=BCS_BoneSpace
    translationMode=BMM_Additive
    translationSpace=BCS_BoneSpace
    alphaBoolBlend {blendOption=Linear}
    boneToModify {boneName=dyn_jacketSide_02_l}
    rotation {pitch=45}
   }
   n120 = Component To Local(ComponentPose=@n119)
   n121 = Get bIsCrouching()
   n122 = Get BwdMovement() -> (BwdMovement:float)
   n123 = Trail controller - Bone: dyn_jacketBack_05_r!(
    ComponentPose=Local To Component(LocalPose=Blend Poses by bool(BlendPose_0=@n120, BlendPose_1=@n101, bActiveValue=@n121) {blendType=Linear})
    Alpha=@n122
    FakeVelocity:Vector="0,0,0"
   ) -> (Pose) {
    bActorSpaceFakeVel=true
    bShowTrailLocation=true
    chainLength=4
    debugLifeTime=0.1
    maxDeltaTime=0.0333
    relaxationSpeedScale=3
    alphaScaleBiasClamp {bInterpResult=true}
    baseJoint {boneName=root}
    planarLimits [
     {drivingBone {boneName=root}, planeTransform {location=<0 0 10>, rotation=<rot 0 0 0>, scale=<1 1 1>}}
     {drivingBone {boneName=spine_03}, planeTransform {location=<0 -20 0>, rotation=<rot 0 15 -89.9999>, scale=<1 1 1>}}
    ]
    rotationLimits [{limitMax=<180 180 180>}, {limitMax=<180 180 180>}, {limitMax=<180 180 180>}, {limitMax=<180 180 180>}]
    rotationOffsets [<0 0 0>, <0 0 0>, <0 0 0>, <0 0 0>]
    trailBone {boneName=dyn_jacketBack_05_r}
    trailRelaxationSpeed {editorCurveData {keys [{arriveTangent=-72.0739, interpMode=RCIM_Cubic, leaveTangent=-72.0739, tangentMode=RCTM_User, time=0, value=35}, <at 1 5 RCIM_Cubic>]}}
   }
   n124 = Trail controller - Bone: dyn_jacketBack_05_l!(ComponentPose=@n123, Alpha=@n122, FakeVelocity:Vector="0,0,0") -> (Pose) {
    bActorSpaceFakeVel=true
    bShowTrailLocation=true
    chainLength=4
    debugLifeTime=0.1
    maxDeltaTime=0.0333
    relaxationSpeedScale=3
    alphaScaleBiasClamp {bInterpResult=true}
    baseJoint {boneName=root}
    planarLimits [
     {drivingBone {boneName=root}, planeTransform {location=<0 0 10>, rotation=<rot 0 0 0>, scale=<1 1 1>}}
     {drivingBone {boneName=spine_03}, planeTransform {location=<0 -20 0>, rotation=<rot 0 15 -89.9999>, scale=<1 1 1>}}
    ]
    rotationLimits [{limitMax=<180 180 180>}, {limitMax=<180 180 180>}, {limitMax=<180 180 180>}, {limitMax=<180 180 180>}]
    rotationOffsets [<0 0 0>, <0 0 0>, <0 0 0>, <0 0 0>]
    trailBone {boneName=dyn_jacketBack_05_l}
    trailRelaxationSpeed {editorCurveData {keys [{arriveTangent=-72.0739, interpMode=RCIM_Cubic, leaveTangent=-72.0739, tangentMode=RCTM_User, time=0, value=35}, <at 1 5 RCIM_Cubic>]}}
   }
   n125 = Get LeftMovement() -> (LeftMovement:float)
   n126 = Trail controller - Bone: dyn_jacketFront_05_r!(ComponentPose=@n124, Alpha=@n125, FakeVelocity:Vector="0,0,0") -> (Pose) {
    bActorSpaceFakeVel=true
    bShowTrailLocation=true
    chainLength=4
    debugLifeTime=0.1
    maxDeltaTime=0.0333
    relaxationSpeedScale=3
    alphaScaleBiasClamp {bInterpResult=true}
    baseJoint {boneName=root}
    planarLimits [
     {drivingBone {boneName=root}, planeTransform {location=<0 0 10>, rotation=<rot 0 0 0>, scale=<1 1 1>}}
     {drivingBone {boneName=spine_03}, planeTransform {location=<0 -20 0>, rotation=<rot 0 15 -89.9999>, scale=<1 1 1>}}
    ]
    rotationLimits [{limitMax=<180 180 180>}, {limitMax=<180 180 180>}, {limitMax=<180 180 180>}, {limitMax=<180 180 180>}]
    rotationOffsets [<0 0 0>, <0 0 0>, <0 0 0>, <0 0 0>]
    trailBone {boneName=dyn_jacketFront_05_r}
    trailRelaxationSpeed {editorCurveData {keys [{arriveTangent=-72.0739, interpMode=RCIM_Cubic, leaveTangent=-72.0739, tangentMode=RCTM_User, time=0, value=35}, <at 1 5 RCIM_Cubic>]}}
   }
   n127 = Trail controller - Bone: dyn_jacketSide_05_r!(ComponentPose=@n126, Alpha=@n125, FakeVelocity:Vector="0,0,0") -> (Pose) {
    bActorSpaceFakeVel=true
    bShowTrailLocation=true
    chainLength=4
    debugLifeTime=0.1
    maxDeltaTime=0.0333
    relaxationSpeedScale=3
    alphaScaleBiasClamp {bInterpResult=true}
    baseJoint {boneName=root}
    planarLimits [
     {drivingBone {boneName=root}, planeTransform {location=<0 0 10>, rotation=<rot 0 0 0>, scale=<1 1 1>}}
     {drivingBone {boneName=spine_03}, planeTransform {location=<0 -20 0>, rotation=<rot 0 15 -89.9999>, scale=<1 1 1>}}
    ]
    rotationLimits [{limitMax=<180 180 180>}, {limitMax=<180 180 180>}, {limitMax=<180 180 180>}, {limitMax=<180 180 180>}]
    rotationOffsets [<0 0 0>, <0 0 0>, <0 0 0>, <0 0 0>]
    trailBone {boneName=dyn_jacketSide_05_r}
    trailRelaxationSpeed {editorCurveData {keys [{arriveTangent=-72.0739, interpMode=RCIM_Cubic, leaveTangent=-72.0739, tangentMode=RCTM_User, time=0, value=35}, <at 1 5 RCIM_Cubic>]}}
   }
   n128 = Get RightMovement() -> (RightMovement:float)
   n129 = Trail controller - Bone: dyn_jacketFront_05_l!(ComponentPose=@n127, Alpha=@n128, FakeVelocity:Vector="0,0,0") -> (Pose) {
    bActorSpaceFakeVel=true
    bShowTrailLocation=true
    chainLength=4
    debugLifeTime=0.1
    maxDeltaTime=0.0333
    relaxationSpeedScale=3
    alphaScaleBiasClamp {bInterpResult=true}
    baseJoint {boneName=root}
    planarLimits [
     {drivingBone {boneName=root}, planeTransform {location=<0 0 10>, rotation=<rot 0 0 0>, scale=<1 1 1>}}
     {drivingBone {boneName=spine_03}, planeTransform {location=<0 -20 0>, rotation=<rot 0 15 -89.9999>, scale=<1 1 1>}}
    ]
    rotationLimits [{limitMax=<180 180 180>}, {limitMax=<180 180 180>}, {limitMax=<180 180 180>}, {limitMax=<180 180 180>}]
    rotationOffsets [<0 0 0>, <0 0 0>, <0 0 0>, <0 0 0>]
    trailBone {boneName=dyn_jacketFront_05_l}
    trailRelaxationSpeed {editorCurveData {keys [{arriveTangent=-72.0739, interpMode=RCIM_Cubic, leaveTangent=-72.0739, tangentMode=RCTM_User, time=0, value=35}, <at 1 5 RCIM_Cubic>]}}
   }
   n130 = Trail controller - Bone: dyn_jacketSide_05_l!(ComponentPose=@n129, Alpha=@n128, FakeVelocity:Vector="0,0,0") -> (Pose) {
    bActorSpaceFakeVel=true
    bShowTrailLocation=true
    chainLength=4
    debugLifeTime=0.1
    maxDeltaTime=0.0333
    relaxationSpeedScale=3
    alphaScaleBiasClamp {bInterpResult=true}
    baseJoint {boneName=root}
    planarLimits [
     {drivingBone {boneName=root}, planeTransform {location=<0 0 10>, rotation=<rot 0 0 0>, scale=<1 1 1>}}
     {drivingBone {boneName=spine_03}, planeTransform {location=<0 -20 0>, rotation=<rot 0 15 -89.9999>, scale=<1 1 1>}}
    ]
    rotationLimits [{limitMax=<180 180 180>}, {limitMax=<180 180 180>}, {limitMax=<180 180 180>}, {limitMax=<180 180 180>}]
    rotationOffsets [<0 0 0>, <0 0 0>, <0 0 0>, <0 0 0>]
    trailBone {boneName=dyn_jacketSide_05_l}
    trailRelaxationSpeed {editorCurveData {keys [{arriveTangent=-72.0739, interpMode=RCIM_Cubic, leaveTangent=-72.0739, tangentMode=RCTM_User, time=0, value=35}, <at 1 5 RCIM_Cubic>]}}
   }
   Save cached pose 'post rigid'(Pose=Blend Poses by bool(
     BlendPose_0=Blend Poses by bool(BlendPose_0=@n69, BlendPose_1=@n73, bActiveValue=@n74) {blendType=Linear}
     BlendPose_1=Component To Local(ComponentPose=@n130)
     bActiveValue=Get bRigidBodyEnabled()
    ) {blendType=Linear}) {cacheName=post rigid}
  ]}
 EventGraph {nodes [
   Event BlueprintUpdateAnimation()
   > Branch(Condition=IsValid(Object=GetOwnerSkeletalMeshComponent()))
   > n1 = Sequence!(execute) -> (then_0, then_1, then_2:Exec, then_3:Exec)
   > n2 = ComplexSimSpaceFunction DEPRECATED!(
    execute
    CharacterPartAnimInstance:Custom Character Part Anim Instance Object Reference=Self-Reference()
    InterpSpeedMax:double=9
    PawnSpeedMax:double=20
    RInterpErrorTolerance:double=1
    DeltaTime:double=0
    DebugDraw:bool=false
    sim_space_rotation_in:Rotator=Get SimSpaceRotation() -> (SimSpaceRotation:Rotator)
    inComponentSpaceIn:bool=Get bInComponentSpace() -> (bInComponentSpace:bool)
    inComponentSpaceLastFrameIn:bool=Get bInComponentSpaceLastFrame() -> (bInComponentSpaceLastFrame:bool)
    lockedOnSimSpaceIn:bool=Get bLockedOnSimSpace() -> (bLockedOnSimSpace:bool)
    InterpSpeedIn:double=Get InterpSpeed() -> (InterpSpeed:float)
   ) -> (
    then
    sim_space_rotation:Rotator
    inComponentSpaceOut:bool
    inComponentSpaceLastFrameOut:bool
    lockedOnSimSpaceOut:bool
    InterpSpeedOut:double
   )
   > Set SimSpaceRotation(SimSpaceRotation:Rotator=@n2) -> (Output_Get:Rotator)
   > Set bInComponentSpace(bInComponentSpace:bool=@n2.inComponentSpaceOut) -> (Output_Get:bool)
   > Set bInComponentSpaceLastFrame(bInComponentSpaceLastFrame:bool=@n2.inComponentSpaceLastFrameOut) -> (Output_Get:bool)
   > Set bLockedOnSimSpace(bLockedOnSimSpace:bool=@n2.lockedOnSimSpaceOut) -> (Output_Get:bool)
   > Set InterpSpeed(InterpSpeed:float=@n2.InterpSpeedOut) -> (Output_Get:float)
   n3 = Gravity Override State Function DEPRECATED(
    execute=@n1.then_1
    CharacterPartAnimInstance=Self-Reference()
    GameTimeInSeconds=GetGameTimeInSeconds()
    GravityOverrideParams=Get GravityOverrideParams() -> (GravityOverrideParams:Gravity Override Params Struct Structure)
   )
   > Set GravityOverride(GravityOverride=@n3)
   n4 = PlayerMovement(execute=@n1.then_2, AnimInstance=Self-Reference(), XDotProductCoefficient=1, YDotProductCoefficient=1, ZDotProductCoefficient=1)
   n5 = MapRangeClamped(Value=@n4.YMovement, InRangeA=0, InRangeB=-1)
   n6 = float * float(A="Absolute (Float)"(A=@n4), B=0.5)
   > Set BwdMovement(BwdMovement:float=MapRangeClamped(
     Value=float + float(A=@n6, B=@n5) {is=title K2Node_CommutativeAssociativeBinaryOperator}
     InRangeA=0
     InRangeB=1.1
     OutRangeA=0.4
     OutRangeB=0
    )) -> (Output_Get:float)
   n7 = MapRangeClamped(Value=@n4, InRangeA=-0.2, InRangeB=1)
   n8 = float * float(A="Absolute (Float)"(A=@n4.YMovement), B=0.2)
   > Set RightMovement(RightMovement:float=MapRangeClamped(Value=float + float(A=@n8, B=@n7), InRangeA=0, InRangeB=1.2, OutRangeB=0.5)) -> (Output_Get:float)
   n9 = MapRangeClamped(Value=@n4, InRangeA=0.2, InRangeB=-1, OutRangeB=0.3)
   n10 = float * float(A="Absolute (Float)"(A=@n4.YMovement), B=0.2)
   > Set LeftMovement(LeftMovement:float=MapRangeClamped(Value=float + float(A=@n10, B=@n9), InRangeA=0, InRangeB=1.2, OutRangeB=0.5)) -> (Output_Get:float)
   n11 = Branch(execute=@n1.then_3, Condition=Get bIsSkyDiving())
   > n12 = Branch(Condition="float > float"(A=Get PawnVelocityZ(), B=50) {is=title K2Node_CallFunction})
   > Set BwdMovement(BwdMovement=float * float(A=Get BwdMovement(), B=0))
   Set BwdMovement(execute=[@n11.else, @n12.else], BwdMovement=float * float(A=Get BwdMovement(), B=1))
   Event BlueprintInitializeAnimation()
  ]}
]
variables [GravityOverrideParams {
  type=GravityOverrideParamsStruct
  default {
   emoteOrMelee {gravityOverride=<0 0 -720>}
   front_End {gravityOverride=<0 0 -720>}
   onGround_Moving {gravityOverride=<0 0 -720>}
   onGround_Standing {gravityOverride=<0 0 -720>}
   skydive_DivingDown {gravityOverride=<0 0 100>}
   skydive_FloatingDown {gravityOverride=<0 0 -140>}
   skydive_GoingUp {gravityOverride=<0 0 100>}
   skydive_Parachuting {gravityOverride=<0 0 -540>}
   swimming {gravityOverride=<0 0 0>}
  }
 }]
