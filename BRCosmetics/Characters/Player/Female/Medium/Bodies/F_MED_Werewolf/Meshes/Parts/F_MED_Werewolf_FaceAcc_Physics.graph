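# Physics asset for the female medium Werewolf face accessory mesh, defining bodies and joints for its dynamic hair bones.
# A two-box main hair chain and single-capsule left and right side locks hang from tiny MSTR root capsules with angularly limited joints.
# Those roots and both clavicle capsules connect to spine_05 with free swing and twist, as does spine_05 to spine_04, forming the upper-torso shapes around the hair.
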PhysicsAsset /BRCosmetics/Characters/Player/Female/Medium/Bodies/F_MED_Werewolf/Meshes/Parts/F_MED_Werewolf_FaceAcc_Physics
bodies {
 clavicle_l [{shapeType=Capsule, length=12.6717, radius=4.8754, shapeName=clavicle_l_capsule, center=<5 0 0>, rotation=<rot 90 0 0>}]
 clavicle_r [{shapeType=Capsule, length=13.1634, radius=4.8754, shapeName=clavicle_r_capsule, center=<-5 0 0>, rotation=<rot 90 0 0>}]
 dyn_hair_main_01 [{
   shapeType=Box
   extentX=7.3717
   extentY=4.5916
   extentZ=20.7421
   shapeName=dyn_hair_main_01_box
   center=<5.0758 0 1.5>
   rotation=<rot -2.5655 0 89.9999>
  }]
 dyn_hair_main_02 [{shapeType=Box, extentX=10.5971, extentY=4.8089, extentZ=15.6392, shapeName=dyn_hair_main_02_box, center=<5.0885 0 1>, rotation=<rot 0 0 90>}]
 dyn_hair_main_MSTR [{shapeType=Capsule, length=0.505, radius=0.505, shapeName=dyn_hair_main_MSTR_capsule, rotation=<rot 0 -90 0>}]
 dyn_hair_side_l [{shapeType=Capsule, length=7.7675, radius=1.9227, shapeName=dyn_hair_side_l_capsule, center=<4 0 2>, rotation=<rot 0 0 90>}]
 dyn_hair_side_l_MSTR [{shapeType=Capsule, length=0.505, radius=0.505, shapeName=dyn_hair_side_l_MSTR_capsule, rotation=<rot 0 -90 0>}]
 dyn_hair_side_r [{shapeType=Capsule, length=7.7675, radius=1.9227, shapeName=dyn_hair_side_r_capsule, center=<4 0 -2>, rotation=<rot 0 0 90>}]
 dyn_hair_side_r_MSTR [{shapeType=Capsule, length=0.505, radius=0.505, shapeName=dyn_hair_side_r_MSTR_capsule, rotation=<rot 0 -90 0>}]
 spine_04 [{shapeType=Capsule, length=7.0044, radius=10.6649, shapeName=spine_04_capsule, center=<2.5931 1.6603 0>, rotation=<rot 0 -92.6845 0>}]
 spine_05 [{shapeType=Capsule, length=17.0459, radius=8.0403, shapeName=spine_05_capsule, center=<1.675 -0.1132 -0.155>}]
}
constraints [
 {bone1Name=dyn_hair_main_01, bone2Name=dyn_hair_main_MSTR, swing1LimitDegrees=45, swing2LimitDegrees=45, twistLimitDegrees=15}
 {bone1Name=dyn_hair_main_02, bone2Name=dyn_hair_main_01, swing1LimitDegrees=45, swing2LimitDegrees=45, twistLimitDegrees=15}
 {
  bone1Name=spine_05
  bone2Name=spine_04
  swing1LimitDegrees=45
  swing1Motion=Free
  swing2LimitDegrees=45
  swing2Motion=Free
  twistLimitDegrees=45
  twistMotion=Free
 }
 {
  bone1Name=dyn_hair_side_r_MSTR
  bone2Name=spine_05
  swing1LimitDegrees=45
  swing1Motion=Free
  swing2LimitDegrees=45
  swing2Motion=Free
  twistLimitDegrees=45
  twistMotion=Free
 }
 {bone1Name=dyn_hair_side_r, bone2Name=dyn_hair_side_r_MSTR, swing1LimitDegrees=35, swing2LimitDegrees=60, twistLimitDegrees=15}
 {
  bone1Name=dyn_hair_side_l_MSTR
  bone2Name=spine_05
  swing1LimitDegrees=45
  swing1Motion=Free
  swing2LimitDegrees=45
  swing2Motion=Free
  twistLimitDegrees=45
  twistMotion=Free
 }
 {bone1Name=dyn_hair_side_l, bone2Name=dyn_hair_side_l_MSTR, swing1LimitDegrees=35, swing2LimitDegrees=60, twistLimitDegrees=15}
 {
  bone1Name=clavicle_l
  bone2Name=spine_05
  swing1LimitDegrees=45
  swing1Motion=Free
  swing2LimitDegrees=45
  swing2Motion=Free
  twistLimitDegrees=45
  twistMotion=Free
 }
 {
  bone1Name=clavicle_r
  bone2Name=spine_05
  swing1LimitDegrees=45
  swing1Motion=Free
  swing2LimitDegrees=45
  swing2Motion=Free
  twistLimitDegrees=45
  twistMotion=Free
 }
 {
  bone1Name=dyn_hair_main_MSTR
  bone2Name=spine_05
  swing1LimitDegrees=45
  swing1Motion=Free
  swing2LimitDegrees=45
  swing2Motion=Free
  twistLimitDegrees=45
  twistMotion=Free
 }
]
