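# Control Rig for the female medium Werewolf face accessory, named for crouch, that adjusts hair and hat bones as the head bends.
# It reads the head's rotation relative to spine_05 and feeds the Euler Z angle through clamped remaps.
# These drive chained Modify Transforms nodes, shifting the left side-hair master bone in X and Z, then shifting the hat master bone likewise and rotating it about Y.
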ControlRigBlueprint /BRCosmetics/Characters/Player/Female/Medium/Bodies/F_MED_Werewolf/Meshes/Parts/F_MED_Werewolf_FaceAcc_crouch_CtrlRig
graphs [
 RigVMFunctionLibrary {links [], nodes []}
 RigVMModel {nodes [
   RigUnit_BeginExecution = &RigUnit_BeginExecution()
   n1 = &RigVMFunction_MathQuaternionToEuler(
    RotationOrder=ZYX
    Value=&RigUnit_GetRelativeTransformForItem(Child=(Type=Bone,Name="head"), Parent=(Type=Bone,Name="spine_05")).RelativeTransform.Rotation
   )
   &RigUnit_ModifyTransforms(
    ItemToModify=((Item=(Type=Bone),Transform=(Rotation=(X=0,Y=0,Z=0,W=1),Translation=(X=0,Y=0,Z=0),Scale3D=(X=1,Y=1,Z=1))),(Item=(Type=Bone,Name="dyn_hair_side_l_MSTR"),Transform=(Rotation=(X=0,Y=0,Z=0,W=1),Translation=(X=-5,Y=0,Z=15),Scale3D=(X=1,Y=1,Z=1))))
    Weight=1
    ItemToModify.1.Transform.Translation.Z=&RigVMFunction_MathFloatRemap(SourceMaximum=-80, SourceMinimum=-20, TargetMaximum=15, bClamp=true, Value=@n1.Result.Z).Result
    ItemToModify.1.Transform.Translation.X=&RigVMFunction_MathFloatRemap(SourceMaximum=-90, SourceMinimum=-20, TargetMaximum=-5, bClamp=true, Value=@n1.Result.Z).Result
    ExecutePin=@RigUnit_BeginExecution.ExecutePin
   )
   n2 = &RigVMFunction_MathQuaternionToEuler(
    RotationOrder=ZYX
    Value=&RigUnit_GetRelativeTransformForItem(Child=(Type=Bone,Name="head"), Parent=(Type=Bone,Name="spine_05")).RelativeTransform.Rotation
   )
   > &RigUnit_ModifyTransforms(
    ItemToModify=((Item=(Type=Bone),Transform=(Rotation=(X=0,Y=0,Z=0,W=1),Translation=(X=0,Y=0,Z=0),Scale3D=(X=1,Y=1,Z=1))),(Item=(Type=Bone,Name="dyn_hat_MSTR"),Transform=(Rotation=(X=0,Y=0,Z=0,W=1),Translation=(X=5,Y=0,Z=6),Scale3D=(X=1,Y=1,Z=1))))
    Weight=1
    ItemToModify.1.Transform.Translation.X=&RigVMFunction_MathFloatRemap(SourceMaximum=-90, SourceMinimum=-40, TargetMaximum=6, bClamp=true, Value=@n2.Result.Z).Result
    ItemToModify.1.Transform.Translation.Z=&RigVMFunction_MathFloatRemap(SourceMaximum=-90, SourceMinimum=-40, TargetMaximum=8, TargetMinimum=0, bClamp=true, Value=@n2.Result.Z).Result
    ItemToModify.1.Transform.Rotation=&RigVMFunction_MathQuaternionFromEuler(
     Euler=(X=0,Y=-20,Z=0)
     Euler.Y=&RigVMFunction_MathFloatRemap(SourceMaximum=-90, SourceMinimum=-40, TargetMaximum=-20, bClamp=true, Value=@n2.Result.Z).Result
    ).Result
   )
  ]}
]
hierarchy [
 root {loc=<0 0 0>, rot=<0 0 0 1>}
 pelvis {parent=root, loc=<0 -0.5301 98.0571>, rot=<0.0127 0.707 -0.0127 -0.707>}
 spine_01 {parent=pelvis, loc=<6.1208 -0.9328 0>, rot=<0 0 0.034 -0.9994>}
 spine_02 {parent=spine_01, loc=<6.5996 -0.2095 0>, rot=<0 0 0.0238 -0.9997>}
 spine_03 {parent=spine_02, loc=<6.4914 -0.1237 0>, rot=<0 0 0.0149 -0.9999>}
 spine_04 {parent=spine_03, loc=<6.9598 0.0794 0>, rot=<0 0 -0.0132 -0.9999>}
 spine_05 {parent=spine_04, loc=<6.9987 0 0>, rot=<0 0 -0.0015 -1>}
 clavicle_l {parent=spine_05, loc=<5.4831 -0.1053 -3.2946>, rot=<0.0465 -0.7289 -0.009 -0.6829>}
 upperarm_l {parent=clavicle_l, loc=<9.8658 -1.2136 -0.4105>, rot=<0.0164 -0.3245 -0.0269 -0.9454>}
 clavicle_r {parent=spine_05, loc=<5.4831 -0.1053 3.2946>, rot=<0.7289 0.0465 -0.6829 0.009>}
 upperarm_r {parent=clavicle_r, loc=<-9.8658 1.2136 0.4104>, rot=<0.0164 -0.3245 -0.0269 -0.9454>}
 neck_01 {parent=spine_05, loc=<11.3138 0.1529 0>, rot=<0 0 -0.2003 -0.9797>}
 neck_02 {parent=neck_01, loc=<4.5627 0 0>, rot=<0 0 0.0504 -0.9987>}
 head {parent=neck_02, loc=<5.3914 0 0>, rot=<0 0 0.115 -0.9934>}
 dyn_hair_main_MSTR {parent=head, loc=<-10.2888 -7.3442 0.7125>, rot=<-0.0184 -0.7344 -0.6782 -0.0204>}
 dyn_hair_main_lookat {parent=dyn_hair_main_MSTR, loc=<0 0 0>, rot=<0 0 0 1>}
 dyn_hair_main_01 {parent=dyn_hair_main_lookat, loc=<0 0 0>, rot=<0.9553 -0.0092 -0.293 0.0386>}
 dyn_hair_main_02 {parent=dyn_hair_main_01, loc=<8.4439 0 0>, rot=<0 -0.2391 0 -0.971>}
 dyn_hair_extra_02 {parent=dyn_hair_main_02, loc=<0.4233 -6.5733 1.4338>, rot=<-0.2261 -0.0239 0.0055 -0.9738>}
 dyn_hair_extra_03 {parent=dyn_hair_main_02, loc=<3.0942 -4.7753 0.2384>, rot=<0 -0.0245 0 -0.9997>}
 dyn_hair_extra_04 {parent=dyn_hair_main_02, loc=<4.6996 0.6371 -0.1301>, rot=<0 -0.0245 0 -0.9997>}
 dyn_hair_extra_05 {parent=dyn_hair_main_02, loc=<2.7386 4.8033 0.9846>, rot=<0.3218 -0.0232 -0.0079 -0.9465>}
 dyn_hair_extra_07 {parent=dyn_hair_main_02, loc=<2.7067 6.486 -0.7561>, rot=<0 -0.0245 0 -0.9997>}
 dyn_hair_extra_06 {parent=dyn_hair_main_02, loc=<2.5088 7.2423 -0.1682>, rot=<0.1304 -0.0243 -0.0032 -0.9912>}
 dyn_hair_extra_01 {parent=dyn_hair_main_01, loc=<1.0198 -6.731 1.5157>, rot=<-0.3492 -0.2451 0.0951 -0.8994>}
 dyn_hair_extra_01_a {parent=dyn_hair_extra_01, loc=<3.2119 -0.6929 3.6084>, rot=<0 0 0 1>}
 dyn_hair_main_01_help {parent=dyn_hair_main_01, loc=<4.2603 -7.0802 0.366>, rot=<0 0 0 -1>}
 dyn_hair_side_r_MSTR {parent=head, loc=<-7.9673 -0.651 4.2026>, rot=<0.0029 -0.7071 -0.0029 -0.7071>}
 dyn_hair_side_r {parent=dyn_hair_side_r_MSTR, loc=<-1.0013 -2.2956 1.3056>, rot=<0 -0.887 0 -0.4617>}
 dyn_hair_side_l_MSTR {parent=head, loc=<-6.2426 -1.1918 -4.2861>, rot=<0.0029 -0.7071 -0.0029 -0.7071>}
 dyn_hair_side_l {parent=dyn_hair_side_l_MSTR, loc=<0.9179 -1.7405 -0.4146>, rot=<0 -0.4617 0 -0.887>}
 dyn_hat_MSTR {parent=head, loc=<0 0 0>, rot=<0.3166 -0.6323 0.6323 -0.3166>}
 dyn_hat {parent=dyn_hat_MSTR, loc=<1.6768 0 3.3217>, rot=<0 0 0 -1>}
 dyn_head {parent=neck_02, loc=<5.3914 0 0>, rot=<0 0 0.115 -0.9934>}
 dyn_hair_front_r_MSTR {parent=dyn_head, loc=<-8.1059 0.4658 6.425>, rot=<0.2092 0.6755 0.6755 0.2092>}
 dyn_hair_front_r_Lookat {parent=dyn_hair_front_r_MSTR, loc=<0 0 0>, rot=<-0.6075 -0.2133 0.2061 -0.7368>}
 dyn_hair_front_r_01 {parent=dyn_hair_front_r_Lookat, loc=<0 0 0>, rot=<0.6075 0.2133 -0.2061 -0.7368>}
 dyn_hair_front_r_02 {parent=dyn_hair_front_r_01, loc=<6.2778 0 0>, rot=<0 0 0 1>}
 dyn_hair_front_l_MSTR {parent=dyn_head, loc=<-8.1059 0.4658 -6.4251>, rot=<0.1977 0.6789 0.6789 0.1977>}
 dyn_hair_front_l_Lookat {parent=dyn_hair_front_l_MSTR, loc=<0 0 0>, rot=<0.6109 -0.2008 -0.1958 -0.7404>}
 dyn_hair_front_l_01 {parent=dyn_hair_front_l_Lookat, loc=<0 0 0>, rot=<-0.6109 0.2008 0.1958 -0.7404>}
 dyn_hair_front_l_02 {parent=dyn_hair_front_l_01, loc=<6.1915 0 0>, rot=<0 0 0 -1>}
 dyn_root_simspace {parent=root, loc=<0 0 0>, rot=<0 0 0 1>}
]
