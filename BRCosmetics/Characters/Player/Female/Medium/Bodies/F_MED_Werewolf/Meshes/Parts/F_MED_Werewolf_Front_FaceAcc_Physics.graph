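# Physics asset for the female medium Werewolf body's front face accessory, defining capsule bodies for its left and right front hair strands.
# Each strand chains a small Lookat body to two hair segments, with set swing limits and tight twist between links.
# Spine and upper-arm capsules cover the torso and shoulders, likely as colliders for the hair.
# The Lookat, upper-arm and spine_04 bodies join spine_05 through free angular constraints.
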PhysicsAsset /BRCosmetics/Characters/Player/Female/Medium/Bodies/F_MED_Werewolf/Meshes/Parts/F_MED_Werewolf_Front_FaceAcc_Physics
bodies {
 dyn_hair_front_l_01 [{shapeType=Capsule, length=3.5915, radius=2.2536, shapeName=dyn_hair_front_l_01_capsule, center=<2.9077 0 0.5>, rotation=<rot 90 0 0>}]
 dyn_hair_front_l_02 [{shapeType=Capsule, length=2.7723, radius=2.0878, shapeName=dyn_hair_front_l_02_capsule, center=<4.0154 -0.5 1>, rotation=<rot 90 0 0>}]
 dyn_hair_front_l_Lookat [{shapeType=Capsule, length=0.505, radius=0.505, shapeName=dyn_hair_front_l_Lookat_capsule, rotation=<rot 0 -90 0>}]
 dyn_hair_front_r_01 [{shapeType=Capsule, length=3.2548, radius=2.8635, shapeName=dyn_hair_front_r_01_capsule, center=<3.0468 0 1>, rotation=<rot 90 0 0>}]
 dyn_hair_front_r_02 [{shapeType=Capsule, length=3.1751, radius=1.8809, shapeName=dyn_hair_front_r_02_capsule, center=<3.5816 -2.5 1>, rotation=<rot 90 0 0>}]
 dyn_hair_front_r_Lookat [{shapeType=Capsule, length=0.505, radius=0.505, shapeName=dyn_hair_front_r_Lookat_capsule, rotation=<rot 0 -90 0>}]
 spine_04 [{shapeType=Capsule, length=7.4066, radius=10.2626, shapeName=spine_04_capsule, center=<2.5514 0.7729 0>, rotation=<rot 0 -92.6845 0>}]
 spine_05 [{shapeType=Capsule, length=17.8161, radius=7.2701, shapeName=spine_05_capsule, center=<2.0326 -0.9928 -0.1551>}]
 upperarm_l [{shapeType=Capsule, length=13.4229, radius=4.1429, shapeName=upperarm_l_capsule, center=<5 -1.25 0.0001>, rotation=<rot 0 -90 90>}]
 upperarm_r [{shapeType=Capsule, length=13.4229, radius=4.1429, shapeName=upperarm_r_capsule, center=<-5 1.25 0.0001>, rotation=<rot 0 -90 90>}]
}
constraints [
 {
  bone1Name=spine_05
  bone2Name=spine_04
  swing1LimitDegrees=45
  swing1Motion=Free
  swing2LimitDegrees=45
  swing2Motion=Free
  twistLimitDegrees=45
  twistMotion=Free
 }
 {bone1Name=dyn_hair_front_l_02, bone2Name=dyn_hair_front_l_01, swing1LimitDegrees=35, swing2LimitDegrees=45, twistLimitDegrees=5}
 {bone1Name=dyn_hair_front_l_01, bone2Name=dyn_hair_front_l_Lookat, swing1LimitDegrees=35, swing2LimitDegrees=45, twistLimitDegrees=5}
 {bone1Name=dyn_hair_front_r_02, bone2Name=dyn_hair_front_r_01, swing1LimitDegrees=35, swing2LimitDegrees=45, twistLimitDegrees=5}
 {bone1Name=dyn_hair_front_r_01, bone2Name=dyn_hair_front_r_Lookat, swing1LimitDegrees=35, swing2LimitDegrees=45, twistLimitDegrees=5}
 {
  bone1Name=dyn_hair_front_r_Lookat
  bone2Name=spine_05
  swing1LimitDegrees=45
  swing1Motion=Free
  swing2LimitDegrees=45
  swing2Motion=Free
  twistLimitDegrees=45
  twistMotion=Free
 }
 {
  bone1Name=dyn_hair_front_l_Lookat
  bone2Name=spine_05
  swing1LimitDegrees=45
  swing1Motion=Free
  swing2LimitDegrees=45
  swing2Motion=Free
  twistLimitDegrees=45
  twistMotion=Free
 }
 {
  bone1Name=upperarm_r
  bone2Name=spine_05
  swing1LimitDegrees=45
  swing1Motion=Free
  swing2LimitDegrees=45
  swing2Motion=Free
  twistLimitDegrees=45
  twistMotion=Free
 }
 {
  bone1Name=upperarm_l
  bone2Name=spine_05
  swing1LimitDegrees=45
  swing1Motion=Free
  swing2LimitDegrees=45
  swing2Motion=Free
  twistLimitDegrees=45
  twistMotion=Free
 }
]
