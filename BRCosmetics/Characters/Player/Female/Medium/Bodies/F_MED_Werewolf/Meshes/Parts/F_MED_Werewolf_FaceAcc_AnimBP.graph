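# Post-process style AnimBlueprint for the female Werewolf face accessory, adding secondary motion to hair and hat.
# Copied body poses pass through constraints and crouch-switched Control Rigs into a cached pose.
# That feeds RigidBody or AnimDynamics main hair, chosen by bRigidBodyEnabled, then AnimDynamics front hair, hat and weighted-constraint extra strands.
# EventGraph varies gravity, zeroes RigidBody inertia with back bling, and drives crouch and skydive bone offsets.
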AnimBlueprint /BRCosmetics/Characters/Player/Female/Medium/Bodies/F_MED_Werewolf/Meshes/Parts/F_MED_Werewolf_FaceAcc_AnimBP
target_skeleton=&/F_MED_Werewolf_FaceAcc_Skeleton
graphs [
 AnimGraph {nodes [
   "Constraint - dyn_hair_main_01_help (P)"(Alpha=1, ConstraintWeights_0=1) {
    alphaBoolBlend {blendOption=Linear}
    boneToModify {boneName=dyn_hair_main_01_help}
    constraintSetup [{targetBone {boneName=clavicle_r}}]
    constraintWeights [1]
   }
   RigidBody!(ComponentPose, Alpha, OverrideWorldGravity:Vector=Get GravityOverride(), ExternalForce) -> (Pose) {
    bClampLinearTranslationLimitToRefPose=true
    bOverrideWorldGravity=true
    overridePhysicsAsset=&/F_MED_Werewolf_Front_FaceAcc_Physics
    simulationSpace=BaseBoneSpace
    alphaBoolBlend {blendOption=Linear}
    baseBoneRef {boneName=dyn_root_simspace}
    componentAppliedLinearAccClamp=<1000 1000 1000>
    componentLinearAccScale=<0.2 0.2 0.3>
    componentLinearVelScale=<0.1 0.1 0.2>
   }
   n1 = RigidBody!(
    ComponentPose=Local To Component(LocalPose=Use cached pose 'prerigid'())
    Alpha
    OverrideWorldGravity=Get GravityOverride_BAck() -> (GravityOverride_BAck:Vector)
    ExternalForce
    ComponentLinearAccScale:Vector=Get AccScale() -> (AccScale:Vector)
    ComponentLinearVelScale:Vector=Get VelScale() -> (VelScale:Vector)
   ) -> (Pose) {
    bClampLinearTranslationLimitToRefPose=true
    bOverrideWorldGravity=true
    overridePhysicsAsset=&/F_MED_Werewolf_FaceAcc_Physics
    simulationSpace=BaseBoneSpace
    alphaBoolBlend {blendOption=Linear}
    baseBoneRef {boneName=spine_05}
    componentAppliedLinearAccClamp=<3000 3000 3000>
    componentLinearAccScale=<0.4 0.4 0.6>
    componentLinearVelScale=<0.4 0.4 0.6>
   }
   n2 = Component To Local(ComponentPose=@n1)
   n3 = "Constraint - dyn_hair_main_01 (P)"(ComponentPose=Local To Component(LocalPose=Use cached pose 'prerigid'()), Alpha=1, ConstraintWeights_0=1) {
    alphaBoolBlend {blendOption=Linear}
    boneToModify {boneName=dyn_hair_main_01}
    constraintSetup [{targetBone {boneName=spine_05}}]
    constraintWeights [1]
   }
   n4 = Anim Dynamics - Bone: dyn_hair_main_02(ComponentPose=@n3, Alpha=1, ExternalForce="0,0,0") {
    angularDampingOverride=0.9999
    angularSpringConstant=100
    bGravityOverrideInSimSpace=true
    bOverrideAngularDamping=true
    simulationSpace=RootRelative
    alphaBoolBlend {blendOption=Linear}
    boundBone {boneName=dyn_hair_main_02}
    componentAppliedLinearAccClamp=<3500 3500 3500>
    componentLinearAccScale=<0.6 0.6 0.8>
    componentLinearVelScale=<0.6 0.6 0.8>
    physicsBodyDefinitions [{
      sphereCollisionRadius=0
      boundBone {boneName=dyn_hair_main_02}
      boxExtents=<20 20 10>
      constraintSetup {angularLimitsMax=<1 -5 10>, angularLimitsMin=<-1 -30 -10>, angularTarget=<0 0 0>}
      localJointOffset=<1 0 0>
     }]
    relativeSpaceBone {boneName=dyn_root_simspace}
   }
   n5 = Component To Local(ComponentPose=@n4)
   n6 = Get bRigidBodyEnabled()
   n7 = "Constraint - dyn_hair_front_r_MSTR (PP)"!(
    ComponentPose=Local To Component(LocalPose=Blend Poses by bool(BlendPose_0=@n2, BlendPose_1=@n5, bActiveValue=@n6) {blendType=Linear})
    Alpha=1
    ConstraintWeights_0=0.8
    ConstraintWeights_1:float=0.2
   ) -> (Pose) {
    alphaBoolBlend {blendOption=Linear}
    boneToModify {boneName=dyn_hair_front_r_MSTR}
    constraintSetup [{targetBone {boneName=clavicle_r}}, {targetBone {boneName=dyn_head}}]
    constraintWeights [1,1]
   }
   n8 = "Constraint - dyn_hair_front_l_MSTR (PP)"!(ComponentPose=@n7, Alpha=1, ConstraintWeights_0=0.8, ConstraintWeights_1:float=0.2) -> (Pose) {
    alphaBoolBlend {blendOption=Linear}
    boneToModify {boneName=dyn_hair_front_l_MSTR}
    constraintSetup [{targetBone {boneName=clavicle_l}}, {targetBone {boneName=dyn_head}}]
    constraintWeights [1,1]
   }
   n9 = Anim Dynamics - Bone: dyn_hair_front_r_01(ComponentPose=@n8, Alpha=1, ExternalForce="0,0,0") {
    angularDampingOverride=0.9999
    angularSpringConstant=100
    bGravityOverrideInSimSpace=true
    bOverrideAngularDamping=true
    simulationSpace=RootRelative
    alphaBoolBlend {blendOption=Linear}
    boundBone {boneName=dyn_hair_front_r_01}
    componentAppliedLinearAccClamp=<3500 3500 3500>
    componentLinearAccScale=<0.6 0.6 0.8>
    componentLinearVelScale=<0.6 0.6 0.8>
    physicsBodyDefinitions [{
      sphereCollisionRadius=0
      boundBone {boneName=dyn_hair_front_r_01}
      boxExtents=<20 20 10>
      constraintSetup {angularLimitsMax=<1 -5 10>, angularLimitsMin=<-1 -30 -10>, angularTarget=<0 0 0>}
      localJointOffset=<1 0 0>
     }]
    relativeSpaceBone {boneName=dyn_root_simspace}
   }
   n10 = Anim Dynamics - Bone: dyn_hair_front_l_01(ComponentPose=@n9, Alpha=1, ExternalForce="0,0,0") {
    angularDampingOverride=0.9999
    angularSpringConstant=100
    bGravityOverrideInSimSpace=true
    bOverrideAngularDamping=true
    simulationSpace=RootRelative
    alphaBoolBlend {blendOption=Linear}
    boundBone {boneName=dyn_hair_front_l_01}
    componentAppliedLinearAccClamp=<3500 3500 3500>
    componentLinearAccScale=<0.6 0.6 0.8>
    componentLinearVelScale=<0.6 0.6 0.8>
    physicsBodyDefinitions [{
      sphereCollisionRadius=0
      boundBone {boneName=dyn_hair_front_l_01}
      boxExtents=<20 20 10>
      constraintSetup {angularLimitsMax=<1 -5 10>, angularLimitsMin=<-1 -30 -10>, angularTarget=<0 0 0>}
      localJointOffset=<1 0 0>
     }]
    relativeSpaceBone {boneName=dyn_root_simspace}
   }
   n11 = Anim Dynamics - Bone: dyn_hair_front_r_02(ComponentPose=@n10, Alpha=1, ExternalForce="0,0,0") {
    angularDampingOverride=0.9999
    angularSpringConstant=100
    bGravityOverrideInSimSpace=true
    bOverrideAngularDamping=true
    simulationSpace=RootRelative
    alphaBoolBlend {blendOption=Linear}
    boundBone {boneName=dyn_hair_front_r_02}
    componentAppliedLinearAccClamp=<3500 3500 3500>
    componentLinearAccScale=<0.6 0.6 0.8>
    componentLinearVelScale=<0.6 0.6 0.8>
    physicsBodyDefinitions [{
      sphereCollisionRadius=0
      boundBone {boneName=dyn_hair_front_r_02}
      boxExtents=<20 20 10>
      constraintSetup {angularLimitsMax=<1 -5 10>, angularLimitsMin=<-1 -30 -10>, angularTarget=<0 0 0>}
      localJointOffset=<1 0 0>
     }]
    relativeSpaceBone {boneName=dyn_root_simspace}
   }
   n12 = Anim Dynamics - Bone: dyn_hair_front_l_02(ComponentPose=@n11, Alpha=1, ExternalForce="0,0,0") {
    angularDampingOverride=0.9999
    angularSpringConstant=100
    bGravityOverrideInSimSpace=true
    bOverrideAngularDamping=true
    simulationSpace=RootRelative
    alphaBoolBlend {blendOption=Linear}
    boundBone {boneName=dyn_hair_front_l_02}
    componentAppliedLinearAccClamp=<3500 3500 3500>
    componentLinearAccScale=<0.6 0.6 0.8>
    componentLinearVelScale=<0.6 0.6 0.8>
    physicsBodyDefinitions [{
      sphereCollisionRadius=0
      boundBone {boneName=dyn_hair_front_l_02}
      boxExtents=<20 20 10>
      constraintSetup {angularLimitsMax=<1 -5 10>, angularLimitsMin=<-1 -30 -10>, angularTarget=<0 0 0>}
      localJointOffset=<1 0 0>
     }]
    relativeSpaceBone {boneName=dyn_root_simspace}
   }
   n13 = Anim Dynamics - Bone: dyn_hat(ComponentPose=@n12, Alpha=1, ExternalForce="0,0,0") {
    angularDampingOverride=0.9999
    angularSpringConstant=300
    bAngularSpring=true
    bGravityOverrideInSimSpace=true
    bOverrideAngularDamping=true
    bUseGravityOverride=true
    simulationSpace=RootRelative
    alphaBoolBlend {blendOption=Linear}
    boundBone {boneName=dyn_hat}
    componentAppliedLinearAccClamp=<5000 5000 5000>
    componentLinearAccScale=<0.8 0.8 0.8>
    componentLinearVelScale=<0.8 0.8 0.8>
    gravityOverride {z=-980}
    physicsBodyDefinitions [{
      sphereCollisionRadius=0
      boundBone {boneName=dyn_hat}
      boxExtents=<20 20 15>
      constraintSetup {angularLimitsMax=<1 20 1>, angularLimitsMin=<-1 -30 -1>, angularTarget=<0 0 0>}
      localJointOffset=<1 0 0>
     }]
    relativeSpaceBone {boneName=dyn_root_simspace}
   }
   n14 = "Constraint - dyn_hair_extra_01 (PPPPPPPP)"!(
    ComponentPose=@n13
    Alpha=1
    ConstraintWeights_0=0.0611
    ConstraintWeights_1:float=0.0341
    ConstraintWeights_2:float=0.2729
    ConstraintWeights_3:float=0.0004
    ConstraintWeights_4:float=0.3916
    ConstraintWeights_5:float=0.2114
    ConstraintWeights_6:float=0.0232
    ConstraintWeights_7:float=0.0053
   ) -> (Pose) {
    alphaBoolBlend {blendOption=Linear}
    boneToModify {boneName=dyn_hair_extra_01}
    constraintSetup [
     {targetBone {boneName=head}}
     {targetBone {boneName=dyn_hair_main_02}}
     {targetBone {boneName=dyn_hair_main_01}}
     {targetBone {boneName=dyn_hair_front_r_01}}
     {targetBone {boneName=spine_05}}
     {targetBone {boneName=dyn_hair_side_r}}
     {targetBone {boneName=neck_02}}
     {targetBone {boneName=neck_01}}
    ]
    constraintWeights [1,1,1,1,1,1,1,1]
   }
   n15 = "Constraint - dyn_hair_extra_02 (PPPPPPPP)"!(
    ComponentPose=@n14
    Alpha=1
    ConstraintWeights_0=0.0007
    ConstraintWeights_1:float=0.1945
    ConstraintWeights_2:float=0.7189
    ConstraintWeights_3:float=0
    ConstraintWeights_4:float=0.0851
    ConstraintWeights_5:float=0
    ConstraintWeights_6:float=0.0006
    ConstraintWeights_7:float=0.0002
   ) -> (Pose) {
    alphaBoolBlend {blendOption=Linear}
    boneToModify {boneName=dyn_hair_extra_02}
    constraintSetup [
     {targetBone {boneName=head}}
     {targetBone {boneName=dyn_hair_main_01}}
     {targetBone {boneName=dyn_hair_main_02}}
     {targetBone {boneName=clavicle_l}}
     {targetBone {boneName=spine_05}}
     {targetBone {boneName=clavicle_r}}
     {targetBone {boneName=neck_02}}
     {targetBone {boneName=neck_01}}
    ]
    constraintWeights [1,1,1,1,1,1,1,1]
   }
   n16 = "Constraint - dyn_hair_extra_03 (PPPPPPPP)"!(
    ComponentPose=@n15
    Alpha=1
    ConstraintWeights_0=0.0001
    ConstraintWeights_1:float=0.0081
    ConstraintWeights_2:float=0.9882
    ConstraintWeights_3:float=0
    ConstraintWeights_4:float=0.0036
    ConstraintWeights_5:float=0
    ConstraintWeights_6:float=0.0001
    ConstraintWeights_7:float=0
   ) -> (Pose) {
    alphaBoolBlend {blendOption=Linear}
    boneToModify {boneName=dyn_hair_extra_03}
    constraintSetup [
     {targetBone {boneName=head}}
     {targetBone {boneName=dyn_hair_main_01}}
     {targetBone {boneName=dyn_hair_main_02}}
     {targetBone {boneName=clavicle_l}}
     {targetBone {boneName=spine_05}}
     {targetBone {boneName=clavicle_r}}
     {targetBone {boneName=neck_02}}
     {targetBone {boneName=neck_01}}
    ]
    constraintWeights [1,1,1,1,1,1,1,1]
   }
   n17 = "Constraint - dyn_hair_extra_04 (P)"!(ComponentPose=@n16, Alpha=1, ConstraintWeights_0=1, ConstraintWeights_1:float=0) -> (Pose) {
    alphaBoolBlend {blendOption=Linear}
    boneToModify {boneName=dyn_hair_extra_04}
    constraintSetup [{targetBone {boneName=dyn_hair_main_02}}]
    constraintWeights [1,0]
   }
   n18 = "Constraint - dyn_hair_extra_05 (PPPPPPPP)"!(
    ComponentPose=@n17
    Alpha=1
    ConstraintWeights_0=0.0004
    ConstraintWeights_1:float=0.0453
    ConstraintWeights_2:float=0.9353
    ConstraintWeights_3:float=0
    ConstraintWeights_4:float=0.0187
    ConstraintWeights_5:float=0
    ConstraintWeights_6:float=0.0002
    ConstraintWeights_7:float=0.0001
   ) -> (Pose) {
    alphaBoolBlend {blendOption=Linear}
    boneToModify {boneName=dyn_hair_extra_05}
    constraintSetup [
     {targetBone {boneName=head}}
     {targetBone {boneName=dyn_hair_main_01}}
     {targetBone {boneName=dyn_hair_main_02}}
     {targetBone {boneName=clavicle_l}}
     {targetBone {boneName=spine_05}}
     {targetBone {boneName=clavicle_r}}
     {targetBone {boneName=neck_02}}
     {targetBone {boneName=neck_01}}
    ]
    constraintWeights [1,1,1,1,1,1,1,1]
   }
   n19 = "Constraint - dyn_hair_extra_06 (PPPPPPPP)"!(
    ComponentPose=@n18
    Alpha=1
    ConstraintWeights_0=0.0001
    ConstraintWeights_1:float=0.0441
    ConstraintWeights_2:float=0.9371
    ConstraintWeights_3:float=0
    ConstraintWeights_4:float=0.0186
    ConstraintWeights_5:float=0
    ConstraintWeights_6:float=0.0001
    ConstraintWeights_7:float=0
   ) -> (Pose) {
    alphaBoolBlend {blendOption=Linear}
    boneToModify {boneName=dyn_hair_extra_06}
    constraintSetup [
     {targetBone {boneName=head}}
     {targetBone {boneName=dyn_hair_main_01}}
     {targetBone {boneName=dyn_hair_main_02}}
     {targetBone {boneName=clavicle_l}}
     {targetBone {boneName=spine_05}}
     {targetBone {boneName=clavicle_r}}
     {targetBone {boneName=neck_02}}
     {targetBone {boneName=neck_01}}
    ]
    constraintWeights [1,1,1,1,1,1,1,1]
   }
   n20 = "Constraint - dyn_hair_extra_07 (PPPPPPPP)"!(
    ComponentPose=@n19
    Alpha=1
    ConstraintWeights_0=0.0001
    ConstraintWeights_1:float=0.0124
    ConstraintWeights_2:float=0.9822
    ConstraintWeights_3:float=0
    ConstraintWeights_4:float=0.0052
    ConstraintWeights_5:float=0
    ConstraintWeights_6:float=0.0001
    ConstraintWeights_7:float=0
   ) -> (Pose) {
    alphaBoolBlend {blendOption=Linear}
    boneToModify {boneName=dyn_hair_extra_07}
    constraintSetup [
     {targetBone {boneName=head}}
     {targetBone {boneName=dyn_hair_main_01}}
     {targetBone {boneName=dyn_hair_main_02}}
     {targetBone {boneName=clavicle_l}}
     {targetBone {boneName=spine_05}}
     {targetBone {boneName=clavicle_r}}
     {targetBone {boneName=neck_02}}
     {targetBone {boneName=neck_01}}
    ]
    constraintWeights [1,1,1,1,1,1,1,1]
   }
   n21 = Anim Dynamics - Bone: dyn_hair_extra_01(ComponentPose=@n20, Alpha=1, ExternalForce="0,0,0") {
    angularDampingOverride=0.9999
    angularSpringConstant=100
    bAngularSpring=true
    bGravityOverrideInSimSpace=true
    bOverrideAngularDamping=true
    simulationSpace=RootRelative
    alphaBoolBlend {blendOption=Linear}
    boundBone {boneName=dyn_hair_extra_01}
    componentAppliedLinearAccClamp=<3500 3500 3500>
    componentLinearAccScale=<1.2 1.2 1.2>
    componentLinearVelScale=<1.2 1.2 1.2>
    physicsBodyDefinitions [{
      sphereCollisionRadius=0
      boundBone {boneName=dyn_hair_extra_01}
      boxExtents=<20 10 10>
      constraintSetup {angularLimitsMax=<1 10 15>, angularLimitsMin=<-1 -30 -15>, angularTarget=<0 0 0>}
      localJointOffset=<2 0 0>
     }]
    relativeSpaceBone {boneName=dyn_root_simspace}
   }
   n22 = Anim Dynamics - Bone: dyn_hair_extra_01_a(ComponentPose=@n21, Alpha=1, ExternalForce="0,0,0") {
    angularDampingOverride=0.9999
    angularSpringConstant=100
    bAngularSpring=true
    bGravityOverrideInSimSpace=true
    bOverrideAngularDamping=true
    simulationSpace=RootRelative
    alphaBoolBlend {blendOption=Linear}
    boundBone {boneName=dyn_hair_extra_01_a}
    componentAppliedLinearAccClamp=<3500 3500 3500>
    componentLinearAccScale=<1.2 1.2 1.2>
    componentLinearVelScale=<1.2 1.2 1.2>
    physicsBodyDefinitions [{
      sphereCollisionRadius=0
      boundBone {boneName=dyn_hair_extra_01_a}
      boxExtents=<20 10 10>
      constraintSetup {angularLimitsMax=<1 10 15>, angularLimitsMin=<-1 -30 -15>, angularTarget=<0 0 0>}
      localJointOffset=<2 0 0>
     }]
    relativeSpaceBone {boneName=dyn_root_simspace}
   }
   n23 = Anim Dynamics - Bone: dyn_hair_extra_02(ComponentPose=@n22, Alpha=1, ExternalForce="0,0,0") {
    angularDampingOverride=0.9999
    angularSpringConstant=100
    bAngularSpring=true
    bGravityOverrideInSimSpace=true
    bOverrideAngularDamping=true
    simulationSpace=RootRelative
    alphaBoolBlend {blendOption=Linear}
    boundBone {boneName=dyn_hair_extra_02}
    componentAppliedLinearAccClamp=<3500 3500 3500>
    componentLinearAccScale=<1.2 1.2 1.2>
    componentLinearVelScale=<1.2 1.2 1.2>
    physicsBodyDefinitions [{
      sphereCollisionRadius=0
      boundBone {boneName=dyn_hair_extra_02}
      boxExtents=<20 10 10>
      constraintSetup {angularLimitsMax=<1 10 15>, angularLimitsMin=<-1 -30 -15>, angularTarget=<0 0 0>}
      localJointOffset=<2 0 0>
     }]
    relativeSpaceBone {boneName=dyn_root_simspace}
   }
   n24 = Anim Dynamics - Bone: dyn_hair_extra_03(ComponentPose=@n23, Alpha=1, ExternalForce="0,0,0") {
    angularDampingOverride=0.9999
    angularSpringConstant=100
    bAngularSpring=true
    bGravityOverrideInSimSpace=true
    bOverrideAngularDamping=true
    simulationSpace=RootRelative
    alphaBoolBlend {blendOption=Linear}
    boundBone {boneName=dyn_hair_extra_03}
    componentAppliedLinearAccClamp=<3500 3500 3500>
    componentLinearAccScale=<1.2 1.2 1.2>
    componentLinearVelScale=<1.2 1.2 1.2>
    physicsBodyDefinitions [{
      sphereCollisionRadius=0
      boundBone {boneName=dyn_hair_extra_03}
      boxExtents=<20 10 10>
      constraintSetup {angularLimitsMax=<1 10 15>, angularLimitsMin=<-1 -30 -15>, angularTarget=<0 0 0>}
      localJointOffset=<2 0 0>
     }]
    relativeSpaceBone {boneName=dyn_root_simspace}
   }
   n25 = Anim Dynamics - Bone: dyn_hair_extra_04(ComponentPose=@n24, Alpha=1, ExternalForce="0,0,0") {
    angularDampingOverride=0.9999
    angularSpringConstant=100
    bAngularSpring=true
    bGravityOverrideInSimSpace=true
    bOverrideAngularDamping=true
    simulationSpace=RootRelative
    alphaBoolBlend {blendOption=Linear}
    boundBone {boneName=dyn_hair_extra_04}
    componentAppliedLinearAccClamp=<3500 3500 3500>
    componentLinearAccScale=<1.2 1.2 1.2>
    componentLinearVelScale=<1.2 1.2 1.2>
    physicsBodyDefinitions [{
      sphereCollisionRadius=0
      boundBone {boneName=dyn_hair_extra_04}
      boxExtents=<20 10 10>
      constraintSetup {angularLimitsMax=<1 10 15>, angularLimitsMin=<-1 -30 -15>, angularTarget=<0 0 0>}
      localJointOffset=<2 0 0>
     }]
    relativeSpaceBone {boneName=dyn_root_simspace}
   }
   n26 = Anim Dynamics - Bone: dyn_hair_extra_05(ComponentPose=@n25, Alpha=1, ExternalForce="0,0,0") {
    angularDampingOverride=0.9999
    angularSpringConstant=100
    bAngularSpring=true
    bGravityOverrideInSimSpace=true
    bOverrideAngularDamping=true
    simulationSpace=RootRelative
    alphaBoolBlend {blendOption=Linear}
    boundBone {boneName=dyn_hair_extra_05}
    componentAppliedLinearAccClamp=<3500 3500 3500>
    componentLinearAccScale=<1.2 1.2 1.2>
    componentLinearVelScale=<1.2 1.2 1.2>
    physicsBodyDefinitions [{
      sphereCollisionRadius=0
      boundBone {boneName=dyn_hair_extra_05}
      boxExtents=<20 10 10>
      constraintSetup {angularLimitsMax=<1 10 15>, angularLimitsMin=<-1 -30 -15>, angularTarget=<0 0 0>}
      localJointOffset=<2 0 0>
     }]
    relativeSpaceBone {boneName=dyn_root_simspace}
   }
   n27 = Anim Dynamics - Bone: dyn_hair_extra_07(ComponentPose=@n26, Alpha=1, ExternalForce="0,0,0") {
    angularDampingOverride=0.9999
    angularSpringConstant=100
    bAngularSpring=true
    bGravityOverrideInSimSpace=true
    bOverrideAngularDamping=true
    simulationSpace=RootRelative
    alphaBoolBlend {blendOption=Linear}
    boundBone {boneName=dyn_hair_extra_07}
    componentAppliedLinearAccClamp=<3500 3500 3500>
    componentLinearAccScale=<1.2 1.2 1.2>
    componentLinearVelScale=<1.2 1.2 1.2>
    physicsBodyDefinitions [{
      sphereCollisionRadius=0
      boundBone {boneName=dyn_hair_extra_07}
      boxExtents=<20 10 10>
      constraintSetup {angularLimitsMax=<1 10 15>, angularLimitsMin=<-1 -30 -15>, angularTarget=<0 0 0>}
      localJointOffset=<2 0 0>
     }]
    relativeSpaceBone {boneName=dyn_root_simspace}
   }
   n28 = Anim Dynamics - Bone: dyn_hair_extra_06(ComponentPose=@n27, Alpha=1, ExternalForce="0,0,0") {
    angularDampingOverride=0.9999
    angularSpringConstant=100
    bAngularSpring=true
    bGravityOverrideInSimSpace=true
    bOverrideAngularDamping=true
    simulationSpace=RootRelative
    alphaBoolBlend {blendOption=Linear}
    boundBone {boneName=dyn_hair_extra_06}
    componentAppliedLinearAccClamp=<3500 3500 3500>
    componentLinearAccScale=<1.2 1.2 1.2>
    componentLinearVelScale=<1.2 1.2 1.2>
    physicsBodyDefinitions [{
      sphereCollisionRadius=0
      boundBone {boneName=dyn_hair_extra_06}
      boxExtents=<20 10 10>
      constraintSetup {angularLimitsMax=<1 10 15>, angularLimitsMin=<-1 -30 -15>, angularTarget=<0 0 0>}
      localJointOffset=<2 0 0>
     }]
    relativeSpaceBone {boneName=dyn_root_simspace}
   }
   n29 = "Transform (Modify) Bone - Bone: dyn_hair_main_01"(ComponentPose=@n28, Alpha=Get hair scale() -> (hair scale:double), Translation="-1,0,0", Rotation="0,0,0", Scale="1,1.000,0.5") {
    scaleMode=BMM_Additive
    translationMode=BMM_Additive
    translationSpace=BCS_BoneSpace
    alphaBoolBlend {blendOption=Linear}
    boneToModify {boneName=dyn_hair_main_01}
    scale {z=0.5}
    translation {x=-1}
   }
   n30 = "Transform (Modify) Bone - Bone: dyn_hair_main_01"!(
    ComponentPose=@n29
    bAlphaBoolEnabled:bool=Get hair helper() -> (hair helper:bool)
    Translation="-1,0,0"
    Rotation="0,0,0"
    Scale="1,1.000,0.5"
   ) -> (Pose) {
    alphaInputType=Bool
    scaleMode=BMM_Additive
    translationMode=BMM_Additive
    translationSpace=BCS_BoneSpace
    alphaBoolBlend {blendInTime=0.1, blendOption=Linear, blendOutTime=0.1}
    boneToModify {boneName=dyn_hair_main_01}
    scale {z=0.5}
    translation {x=-1}
   }
   Output Pose(Result=Component To Local(ComponentPose=@n30)) {layerGroup=None}
   n31 = Copy Pose From Mesh(SourceMeshComponent=Get Skel_Mesh()) {bUseAttachedParent=true}
   n32 = Copy Pose From Mesh() {bUseAttachedParent=true}
   n33 = "Constraint - dyn_head (P)"(
    ComponentPose=Local To Component(LocalPose=Layered blend per bone(BasePose=@n32, BlendPoses_0=@n31) {layerSetup [{branchFilters [{boneName=upperarm_l}, {boneName=upperarm_r}]}]})
    Alpha=0.5
    ConstraintWeights_0=1
   ) {alphaBoolBlend {blendOption=Linear}, boneToModify {boneName=dyn_head}, constraintSetup [{targetBone {boneName=head}}], constraintWeights [1]}
   n34 = "Constraint - dyn_hat_MSTR (P)"(ComponentPose=@n33, Alpha=1, ConstraintWeights_0=1) {
    alphaBoolBlend {blendOption=Linear}
    boneToModify {boneName=dyn_hat_MSTR}
    constraintSetup [{targetBone {boneName=head}}]
    constraintWeights [1]
   }
   n35 = "Constraint - dyn_hair_main_MSTR (P)"(ComponentPose=@n34, Alpha=1, ConstraintWeights_0=1) {
    alphaBoolBlend {blendOption=Linear}
    boneToModify {boneName=dyn_hair_main_MSTR}
    constraintSetup [{targetBone {boneName=spine_05}}]
    constraintWeights [1]
   }
   n36 = "Constraint - dyn_hair_side_l_MSTR (PPPP)"!(
    ComponentPose=@n35
    Alpha=1
    ConstraintWeights_0=0.4002
    ConstraintWeights_1:float=0.2614
    ConstraintWeights_2:float=0.2097
    ConstraintWeights_3:float=0.1286
   ) -> (Pose) {
    alphaBoolBlend {blendOption=Linear}
    boneToModify {boneName=dyn_hair_side_l_MSTR}
    constraintSetup [{targetBone {boneName=neck_02}}, {targetBone {boneName=head}}, {targetBone {boneName=neck_01}}, {targetBone {boneName=spine_05}}]
    constraintWeights [1,1,1,1]
   }
   n37 = "Constraint - dyn_hair_side_r_MSTR (PPPPP)"!(
    ComponentPose=@n36
    Alpha=1
    ConstraintWeights_0=0.001
    ConstraintWeights_1:float=0.2893
    ConstraintWeights_2:float=0.1648
    ConstraintWeights_3:float=0.1691
    ConstraintWeights_4:float=0.3758
   ) -> (Pose) {
    alphaBoolBlend {blendOption=Linear}
    boneToModify {boneName=dyn_hair_side_r_MSTR}
    constraintSetup [
     {targetBone {boneName=clavicle_r}}
     {targetBone {boneName=neck_02}}
     {targetBone {boneName=head}}
     {targetBone {boneName=neck_01}}
     {targetBone {boneName=spine_05}}
    ]
    constraintWeights [1,1,1,1,1]
   }
   n38 = Look At - Bone: dyn_hair_main_lookat(ComponentPose=@n37, Alpha=1) {
    alphaBoolBlend {blendOption=Linear}
    boneToModify {boneName=dyn_hair_main_lookat}
    lookAtLocation {x=0}
    lookAtTarget {boneReference {boneName=dyn_root_simspace}}
    lookAt_Axis {axis=<1 0>}
   }
   n39 = Control Rig!(Source=Component To Local(ComponentPose=@n38), bAlphaBoolEnabled:bool=NOT Boolean(A=Get bIsCrouching())) -> (Pose) {
    alphaInputType=Bool
    alphaBoolBlend {blendInTime=0.3, blendOption=Linear, blendOutTime=0.1}
    controlRigAssetReference {blueprintRigClass=&/F_MED_Werewolf_FaceAcc_CtrlRig.F_MED_Werewolf_FaceAcc_CtrlRig_C}
   }
   n40 = Control Rig(Source=@n39, bAlphaBoolEnabled=Get hat() -> (hat:bool)) {
    alphaInputType=Bool
    alphaBoolBlend {blendInTime=0.3, blendOption=Linear, blendOutTime=0.1}
    controlRigAssetReference {blueprintRigClass=&/F_MED_Werewolf_FaceAcc_crouch_CtrlRig.F_MED_Werewolf_FaceAcc_crouch_CtrlRig_C}
   }
   n41 = "Transform (Modify) Bone - Bone: dyn_hat_MSTR"!(
    ComponentPose=Local To Component(LocalPose=@n40)
    bAlphaBoolEnabled:bool=Get hat_rot() -> (hat_rot:bool)
    Translation="1,0,0"
    Rotation="0,15,15"
    Scale="1,1,1"
   ) -> (Pose) {
    alphaInputType=Bool
    rotationMode=BMM_Additive
    rotationSpace=BCS_BoneSpace
    translationMode=BMM_Additive
    translationSpace=BCS_BoneSpace
    alphaBoolBlend {blendInTime=0.3, blendOption=Linear, blendOutTime=0.3}
    boneToModify {boneName=dyn_hat_MSTR}
    rotation {roll=15, yaw=15}
    translation {x=1}
   }
   n42 = "Transform (Modify) Bone - Bone: dyn_hair_main_01_help"(ComponentPose=@n41, Alpha=1, Translation="0,0,3.5", Rotation="0,0,0", Scale="1,1,1") {
    translationMode=BMM_Additive
    translationSpace=BCS_BoneSpace
    alphaBoolBlend {blendOption=Linear}
    boneToModify {boneName=dyn_hair_main_01_help}
    translation {z=3.5}
   }
   Save cached pose 'prerigid'(Pose=Component To Local(ComponentPose=@n42)) {cacheName=prerigid}
  ]}
 EventGraph {nodes [
   Event OnCharacterCustomizationLoaded()
   > Delay(Duration=0.1)
   > ResetDynamics(InTeleportType=TeleportPhysics)
   > ResetDynamics(InTeleportType=ResetPhysics)
   Event BlueprintUpdateAnimation()
   > Get!(execute:Exec) -> (then:Exec, else:Exec, Skel_Mesh:Skeletal Mesh Component Object Reference)
   > n1 = Sequence!(execute) -> (then_0, then_1, then_2:Exec, then_3:Exec, then_4:Exec, then_5:Exec, then_6:Exec)
   > n2 = Gravity Override State Function DEPRECATED(CharacterPartAnimInstance=Self-Reference(), GameTimeInSeconds=GetGameTimeInSeconds(), GravityOverrideParams=Get Gravity Override Params())
   > Set GravityOverride(GravityOverride=@n2)
   > n3 = Branch(Condition=OR Boolean(A=Get WearingCape() -> (WearingCape:bool), B=Get wearingBackpack() -> (wearingBackpack:bool)))
   > Set GravityOverride_BAck(GravityOverride_BAck:Vector="0, 0,-980") -> (Output_Get:Vector)
   Set GravityOverride_BAck(execute=@n3.else, GravityOverride_BAck=@n2)
   n4 = ComplexSimSpaceFunction DEPRECATED!(
    execute=@n1.then_1
    CharacterPartAnimInstance:Custom Character Part Anim Instance Object Reference=Self-Reference()
    InterpSpeedMax:double=9
    PawnSpeedMax:double=20
    RInterpErrorTolerance:double=1
    DeltaTime:double=0
    DebugDraw:bool=false
    sim_space_rotation_in:Rotator=Get sim_space_rotation() -> (sim_space_rotation:Rotator)
    inComponentSpaceIn:bool=Get InComponentSpace() -> (InComponentSpace:bool)
    inComponentSpaceLastFrameIn:bool=Get InComponentSpaceLastFrame() -> (InComponentSpaceLastFrame:bool)
    lockedOnSimSpaceIn:bool=Get LockedOnSimSpace() -> (LockedOnSimSpace:bool)
    InterpSpeedIn:double=Get InterpSpeed() -> (InterpSpeed:double)
   ) -> (
    then
    sim_space_rotation:Rotator
    inComponentSpaceOut:bool
    inComponentSpaceLastFrameOut:bool
    lockedOnSimSpaceOut:bool
    InterpSpeedOut:double
   )
   > Set sim_space_rotation(sim_space_rotation:Rotator=@n4) -> (Output_Get:Rotator)
   > Set InComponentSpace(InComponentSpace:bool=@n4.inComponentSpaceOut) -> (Output_Get:bool)
   > Set InComponentSpaceLastFrame(InComponentSpaceLastFrame:bool=@n4.inComponentSpaceLastFrameOut) -> (Output_Get:bool)
   > Set LockedOnSimSpace(LockedOnSimSpace:bool=@n4.lockedOnSimSpaceOut) -> (Output_Get:bool)
   > Set InterpSpeed(InterpSpeed:double=@n4.InterpSpeedOut) -> (Output_Get:double)
   n5 = CheckForBackbling()
   n6 = AND Boolean(A=@n5, B=NOT Boolean(A=@n5.bIsBackblingLong))
   Set wearingBackpack(execute=@n1.then_2, wearingBackpack:bool=@n6) -> (Output_Get:bool)
   > Set isWearingBackpack(isWearingBackpack:double="To Float (Boolean)"(InBool=@n6)) -> (Output_Get:double)
   > Set isNotWearingBackbling(isNotWearingBackbling:double="To Float (Boolean)"(InBool=NOT Boolean(A=@n5))) -> (Output_Get:double)
   > Set isWearingCape(isWearingCape:double="To Float (Boolean)"(InBool=@n5.bIsBackblingLong)) -> (Output_Get:double)
   > Set WearingCape(WearingCape:bool=AND Boolean(A=@n5, B=@n5.bIsBackblingLong)) -> (Output_Get:bool)
   n7 = Branch(execute=@n1.then_3, Condition=OR Boolean(A=Get WearingCape(), B=Get wearingBackpack()))
   > Set AccScale(AccScale:Vector="0, 0, 0") -> (Output_Get:Vector)
   > Set VelScale(VelScale:Vector="0, 0, 0") -> (Output_Get:Vector)
   Set AccScale(execute=@n7.else, AccScale="0.3,0.3,0.5")
   > Set VelScale(VelScale="0.3,0.3,0.5")
   n8 = Branch(execute=@n1.then_4, Condition=Get bIsCrouching())
   > Set hat(hat:bool=true) -> (Output_Get:bool)
   > n9 = Branch(Condition=Get bIsMoving2D())
   > Set hat_rot(hat_rot:bool=true) -> (Output_Get:bool)
   Set hat_rot(execute=@n9.else, hat_rot=false)
   Set hat(execute=@n8.else, hat=false)
   n10 = Branch(execute=@n1.then_5, Condition=AND Boolean(A=Get bIsCrouching(), B=Get bIsMoving2D()))
   > Set hair scale(hair scale:double=GetAlphaFromReferenceTransformDelta(SkeletalMeshComponent=Get Skel_Mesh(), BoneName=head, BaseName=spine_05, InputRange=(X=-40,Y=-70), TwistAxis="1,0,0")) -> (Output_Get:double)
   Set hair scale(execute=@n10.else, hair scale=0)
   n11 = Get wearingBackpack()
   n12 = Get WearingCape()
   n13 = Branch(execute=@n1.then_6, Condition=AND Boolean(A=OR Boolean(A=@n12, B=@n11), B=Get bIsSkyDiving()))
   > Set hair helper(hair helper:bool=true) -> (Output_Get:bool)
   Set hair helper(execute=@n13.else, hair helper=false)
   Event BlueprintInitializeAnimation()
   > Set Skel_Mesh(Skel_Mesh=GetOwnerSkeletalMeshComponent())
  ]}
]
variables [
 AccScale {type=Vector}
 backpack {type=number}
 Gravity Override Params {
  type=GravityOverrideParamsStruct
  default {
   emoteOrMelee {gravityOverride=<0 0 -720>}
   front_End {gravityOverride=<0 0 -720>}
   onGround_Moving {gravityOverride=<0 0 -720>}
   onGround_Standing {gravityOverride=<0 0 -720>}
   skydive_DivingDown {gravityOverride=<0 0 0>}
   skydive_FloatingDown {gravityOverride=<0 0 -240>}
   skydive_GoingUp {gravityOverride=<0 0 100>}
   skydive_Parachuting {gravityOverride=<0 0 -540>}
   swimming {gravityOverride=<0 0 0>}
  }
 }
 GravityOverride {type=Vector}
 GravityOverride_BAck {type=Vector}
 hair helper {type=boolean}
 hair scale {type=number}
 hat {type=boolean}
 hat_rot {type=boolean}
 InComponentSpace {type=boolean}
 InComponentSpaceLastFrame {type=boolean}
 InterpSpeed {type=number}
 isNotWearingBackbling {type=number}
 isWearingBackpack {type=number}
 isWearingCape {type=number}
 LockedOnSimSpace {type=boolean}
 sim_space_rotation {type=Rotator}
 Skel_Mesh {type=SkeletalMeshComponent}
 VelScale {type=Vector}
 wearingBackpack {type=boolean}
 WearingCape {type=boolean}
]
